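# Control Rig for the SteelyKendama medium male player body that adds corrective deformation and poses accessory bones such as the bag, belt, ropes and shuriken.
# It runs a shared DeformRig function, then uses spherical pose readers and calf-yaw remaps to weight transform offsets on thighs, knees, feet, upper arms and accessories.
# Skydive-floating and DBNO states are inferred from pelvis rotation, height and velocity to choose bag poses, alongside a height-based crouch adjustment that sets CrouchIdleAlpha.
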ControlRigBlueprint /BRCosmetics/Characters/Player/Male/Medium/Bodies/M_MED_SteelyKendama/Meshes/M_MED_SteelyKendama_CtrlRig
graphs [
 ContainedGraph {nodes [
   n1 = RigVMFunctionEntryNode()
   n2 = &RigUnit_GetTransform(Space=LocalSpace, Item=@n1.Item)
   n3 = &RigUnit_GetTransform(Space=LocalSpace, bInitial=true, Item=@n1.Item)
   > RigVMFunctionReturnNode(OutTransform=/Script/RigVM.RigVMFunction_MathTransformToEulerTransform(Value=&RigVMFunction_MathTransformMakeRelative(Global=@n2.Transform, Parent=@n3.Transform).Local) {is=struct RigVMUnitNode, template="Cast::Execute(in Value,out Result)"}.Result)
  ]}
 ContainedGraph {nodes [
   n1 = RigVMFunctionEntryNode()
   n2 = &RigUnit_GetTransform(Space=LocalSpace, Item=@n1.Item)
   n3 = &RigUnit_GetTransform(Space=LocalSpace, bInitial=true, Item=@n1.Item)
   > RigVMFunctionReturnNode(OutTransform=&RigVMFunction_MathTransformToEulerTransform(Value=&RigVMFunction_MathTransformMakeRelative(Global=@n2.Transform, Parent=@n3.Transform).Local) {template="Cast::Execute(in Value,out Result)"}.Result)
  ]}
 Multiply_4_1_ContainedGraph {nodes [
   n1 = RigVMFunctionEntryNode()
   RigVMFunctionReturnNode(Result=&RigVMFunction_MathDoubleMul(B=@n1.C, A=&RigVMFunction_MathDoubleMul(A=@n1.A, B=@n1.B).Result).Result)
  ]}
 Multiply_4_ContainedGraph {nodes [
   n1 = RigVMFunctionEntryNode()
   Multiply = &RigVMFunction_MathDoubleMul(A=@n1.A, B=@n1.B)
   RigVMFunctionReturnNode(Result=&RigVMFunction_MathDoubleMul(B=@n1.C, A=@Multiply.Result).Result)
  ]}
 New Function_ContainedGraph {nodes [
   n1 = RigVMFunctionEntryNode()
   Cast = /Script/RigVM.RigVMFunction_MathQuaternionToRotator(Value=&RigUnit_GetRelativeTransformForItem(bChildInitial=False, bParentInitial=true, Child=@n1.Object, Parent=@n1.Object).RelativeTransform.Rotation) {is=struct RigVMUnitNode, template="Cast::Execute(in Value,out Result)"}
   n2 = RigVMDispatchNode(Struct=@Cast.Result) {template="DISPATCH_RigVMDispatch_BreakStruct(in Struct,out Elements)"}
   Make = RigVMDispatchNode(
    Elements=(Pitch=0,Yaw=0,Roll=0)
    Elements.Roll=&RigVMFunction_MathDoubleMul(A=@n2.Elements.Roll, B=@n1.Factor).Result
    Elements.Pitch=&RigVMFunction_MathDoubleMul(A=@n2.Elements.Pitch, B=@n1.Factor).Result
    Elements.Yaw=&RigVMFunction_MathDoubleMul(A=@n2.Elements.Yaw, B=@n1.Factor).Result
   ) {template="DISPATCH_RigVMDispatch_MakeStruct(in Elements,out Struct)"}
   > RigVMFunctionReturnNode(Result=/Script/RigVM.RigVMFunction_MathQuaternionFromRotatorV2(Value=@Make.Struct) {is=struct RigVMUnitNode, template="Cast::Execute(in Value,out Result)"}.Result)
  ]}
 RigVMFunctionLibrary {nodes [
   DeformRig_GetDeltaRefTransform_FNC = RigVMCollapseNode(ExecuteContext=(), Item=(Type=Bone))
   DeformRig_GetDeltaRefTransform_FNC_1 = RigVMCollapseNode(ExecuteContext=(), Item=(Type=Bone))
   Get Relative Rotation self initial = RigVMCollapseNode(Factor=-0.5, Object=(Type=Bone,Name="C_Tail_A_5_Jnt"))
  ]}
 RigVMFunction_MathBoolAnd_1_ContainedGraph {nodes [
   n1 = RigVMFunctionEntryNode()
   RigVMFunction_MathBoolAnd = &RigVMFunction_MathBoolAnd(A=@n1.A, B=@n1.B)
   RigVMFunctionReturnNode(Result=&RigVMFunction_MathBoolAnd(B=@n1.C, A=@RigVMFunction_MathBoolAnd.Result).Result)
  ]}
 RigVMFunction_Sequence_1_1_ContainedGraph {nodes [
   n1 = RigVMFunctionEntryNode()
   n2 = &RigVMFunction_Sequence(ExecuteContext=@n1.ExecuteContext)
   n3 = &RigVMFunction_Sequence(ExecuteContext=@n2.B)
   n4 = &RigVMFunction_Sequence(ExecuteContext=@n3.B)
   n5 = &RigVMFunction_Sequence(ExecuteContext=@n4.B)
   n6 = &RigVMFunction_Sequence(ExecuteContext=@n5.B)
   n7 = &RigVMFunction_Sequence(ExecuteContext=@n6.B)
   RigVMFunctionReturnNode(A=@n2.A, B=@n3.A, C=@n4.A, D=@n5.A, E=@n6.A, F=@n7.A, G=@n7.B)
  ]}
 RigVMFunction_Sequence_3_1_ContainedGraph {nodes [
   n1 = RigVMFunctionEntryNode()
   n2 = &RigVMFunction_Sequence(ExecuteContext=@n1.ExecuteContext)
   n3 = &RigVMFunction_Sequence(ExecuteContext=@n2.B)
   n4 = &RigVMFunction_Sequence(ExecuteContext=@n3.B)
   RigVMFunctionReturnNode(A=@n2.A, B=@n3.A, C=@n4.A, D=@n4.B)
  ]}
 RigVMModel {nodes [
   RigUnit_BeginExecution = &RigUnit_BeginExecution()
   FN_DeformRig_FNC = RigVMFunctionReferenceNode(
    DeformRig_ Curve="(deformRig_l_upperarm_up_alpha_72_#1=False,deformRig_r_upperarm_up_alpha_73_#2=False,deformRig_l_upperarm_fr_alpha_74_#3=False,deformRig_r_upperarm_fr_alpha_75_CB51B8 ...(truncated)"
    DeformRig_Bool="(upperarm_1_#4=True,elbow_3_#5=True,wrist_7_#6=True,groin_8_#7=True,glute_17_#8=True,knee_9_40C408324CA261 ...(truncated)"
    DeformRig_Debug=(upperarm_debug_18_#4=False,wrist_debug_20_#6=False,lat_debug_26_#9=false,groinGlute_debug_27_#7=False)
    DeformRig_Value="(upperarm_up_18_#10=(Rotation=(X=0,Y=0,Z=0,W=1),Translation=(X=0,Y=0,Z=8),Scale3D=(X=0.7,Y=1,Z=1)),upperarm_fwd_fr_19_#11=(Rotation=(X=0,Y=0,Z=0,W=1),Translation=(X=0,Y=10,Z=0),Scale3D=(X=0.5,Y=1,Z= ...(truncated)"
    ExecuteContext=@RigUnit_BeginExecution.ExecutePin
   )
   > RigVMFunction_Sequence_1 = RigVMAggregateNode()
   n1 = &RigVMFunction_Sequence(ExecuteContext=@RigVMFunction_Sequence_1.B)
   n2 = &RigUnit_SphericalPoseReader(
    ActiveRegionScaleFactors=(PositiveWidth=3.5,NegativeWidth=0.5,PositiveHeight=3,NegativeHeight=3)
    ActiveRegionSize=0.1
    Debug=(bDrawDebug=False,DebugScale=25,DebugSegments=20,DebugThickness=0.25)
    DriverAxis=(X=-1,Y=0,Z=0)
    DriverItem=(Type=Bone,Name="thigh_l")
    FalloffRegionScaleFactors=(PositiveWidth=1,NegativeWidth=1.6,PositiveHeight=2.5,NegativeHeight=1)
    FalloffSize=0.45
    OptionalParentItem=(Type=Bone)
    RotationOffset=(X=90,Y=0,Z=-35)
    ExecutePin=@n1.A
   )
   > &RigUnit_ModifyTransforms(
    ItemToModify="((Item=(Type=Bone,Name=\"thigh_twist_01_l\"),Transform=(Rotation=(X=0,Y=0,Z=0.1189,W=0.9929),Translation=(X=-0.1623,Y=-4.9518,Z=0),Scale3D=(X=0.9048,Y=1,Z=1))),(Item=(Type=Bone,Name=\"dyn_lope_l_front_01\"),Transform=(Rotation=(X=0.0282,Y=0,Z=0,W=0.9996),Trans ...(truncated)"
    WeightMinimum=0
    Weight=@n2.OutputParam
   )
   > DeformRig_GetDeltaRefTransform_FNC = RigVMFunctionReferenceNode(Item=(Type=Bone,Name="calf_l"))
   > &RigUnit_ModifyTransforms(
    ItemToModify=((Item=(Type=Bone,Name="thigh_twist_01_l"),Transform=(Rotation=(X=0,Y=0,Z=0.0291,W=0.9996),Translation=(X=7.6439,Y=-1.4299,Z=0),Scale3D=(X=0.9374,Y=1.1638,Z=1))))
    WeightMinimum=0
    Weight=&RigVMFunction_MathFloatRemap(SourceMaximum=100, SourceMinimum=10, bClamp=true, Value=@DeformRig_GetDeltaRefTransform_FNC.OutTransform.Rotation.Yaw).Result
   )
   > /Script/ControlRig.RigUnit_PositionConstraintLocalSpaceOffset(
    Child=(Type=Bone,Name="dyn_rope_m_front_01")
    Filter=(bX=True,bY=True,bZ=True)
    Parents=((Item=(Type=Bone,Name="dyn_lope_lm_front_01"),Weight=1),(Item=(Type=Bone,Name="dyn_rope_rm_front_01"),Weight=1))
    Weight=1
    bMaintainOffset=True
   ) {is=struct RigVMUnitNode}
   > &RigUnit_ParentConstraint(
    AdvancedSettings=(InterpolationType=Average,RotationOrderForFilter=XZY)
    Child=(Type=Bone,Name="dyn_rope_m_front_01")
    Parents=((Item=(Type=Bone,Name="dyn_lope_lm_front_01"),Weight=1),(Item=(Type=Bone,Name="dyn_rope_rm_front_01"),Weight=1))
    Weight=1
   )
   n3 = &RigUnit_SphericalPoseReader(
    ActiveRegionScaleFactors=(PositiveWidth=3.5,NegativeWidth=0.5,PositiveHeight=3,NegativeHeight=3)
    ActiveRegionSize=0.1
    Debug=(bDrawDebug=False,DebugScale=25,DebugSegments=20,DebugThickness=0.25)
    DriverAxis=(X=1,Y=0,Z=0)
    DriverItem=(Type=Bone,Name="thigh_r")
    FalloffRegionScaleFactors=(PositiveWidth=1,NegativeWidth=1.6,PositiveHeight=2.5,NegativeHeight=1)
    FalloffSize=0.45
    FlipHeightScaling=false
    FlipWidthScaling=true
    OptionalParentItem=(Type=Bone)
    RotationOffset=(X=-90,Y=0,Z=-35)
    ExecutePin=@n1.B
   )
   > &RigUnit_ModifyTransforms(
    ItemToModify="((Item=(Type=Bone,Name=\"thigh_twist_01_r\"),Transform=(Rotation=(X=0,Y=0,Z=0.1189,W=0.9929),Translation=(X=0.1623,Y=4.9518,Z=0),Scale3D=(X=0.9048,Y=1,Z=1))),(Item=(Type=Bone,Name=\"dyn_rope_rm_front_01\"),Transform=(Rotation=(X=-0.1084,Y=0,Z=0,W=0.9941),Trans ...(truncated)"
    WeightMinimum=0
    Weight=@n3.OutputParam
   )
   > DeformRig_GetDeltaRefTransform_FNC_1 = RigVMFunctionReferenceNode(Item=(Type=Bone,Name="calf_r"))
   > &RigUnit_ModifyTransforms(
    ItemToModify=((Item=(Type=Bone,Name="thigh_twist_01_r"),Transform=(Rotation=(X=0,Y=0,Z=0.0291,W=0.9996),Translation=(X=-7.6439,Y=1.4299,Z=0),Scale3D=(X=0.9374,Y=1.1638,Z=1))))
    WeightMinimum=0
    Weight=&RigVMFunction_MathFloatRemap(SourceMaximum=100, SourceMinimum=10, bClamp=true, Value=@DeformRig_GetDeltaRefTransform_FNC_1.OutTransform.Rotation.Yaw).Result
   )
   RigVMFunction_Sequence_3 = RigVMAggregateNode(ExecuteContext=@RigVMFunction_Sequence_1.C)
   n4 = RigVMVariableNode() {variable=Is pot SkydiveFloating}
   RigVMFunction_ControlFlowBranch = &RigVMFunction_ControlFlowBranch(ExecuteContext=@RigVMFunction_Sequence_3.A, Condition=@n4.Value)
   n5 = RigVMVariableNode() {variable=DBNO Alpha}
   n6 = &RigVMFunction_ControlFlowBranch(
    Condition=RigVMDispatchNode(B=1, A=@n5.Value) {template="DISPATCH_RigVMDispatch_CoreEquals(in A,in B,out Result)"}.Result
    ExecuteContext=@RigVMFunction_ControlFlowBranch.False
   )
   n7 = &RigVMFunction_AlphaInterp(InterpSpeedDecreasing=15, InterpSpeedIncreasing=15, Scale=1, bClampResult=False, Value=@n5.Value)
   &RigUnit_ModifyTransforms(
    ItemToModify="((Item=(Type=Bone,Name=\"dyn_bagDrive_master\"),Transform=(Rotation=(X=-0.1972,Y=0.4526,Z=0.0303,W=0.8691),Translation=(X=0.0726,Y=6.6299,Z=6.9697),Scale3D=(X=1,Y=1,Z=1))),(Item=(Type=Bone,Name=\"dyn_beltFront_anchor_c\"),Transform=(Rotation=(X=0.0576,Y=0.0003 ...(truncated)"
    WeightMinimum=0
    Weight=@n7.Result
    ExecutePin=@n6.True
   )
   Cast = &RigVMFunction_MathQuaternionToRotator(Value=&RigUnit_GetRelativeTransformForItem(Child=(Type=Bone,Name="thigh_r"), Parent=(Type=Bone,Name="thigh_r"), bChildInitial=False, bParentInitial=true).RelativeTransform.Rotation) {template="Cast::Execute(in Value,out Result)"}
   n8 = RigVMDispatchNode(Struct=@Cast.Result) {template="DISPATCH_RigVMDispatch_BreakStruct(in Struct,out Elements)"}
   Make = RigVMDispatchNode(
    Elements=(Pitch=0,Yaw=0,Roll=0)
    Elements.Roll=&RigVMFunction_MathDoubleMul(B=0.1, A=@n8.Elements.Roll).Result
    Elements.Pitch=&RigVMFunction_MathDoubleMul(B=0.6, A=@n8.Elements.Pitch).Result
    Elements.Yaw=&RigVMFunction_MathDoubleMul(B=0.7, A=@n8.Elements.Yaw).Result
   ) {template="DISPATCH_RigVMDispatch_MakeStruct(in Elements,out Struct)"}
   &RigUnit_ModifyTransforms(
    ItemToModify=((Item=(Type=Bone,Name="dyn_bagDrive_master"),Transform=(Rotation=(X=0,Y=0,Z=0,W=1),Translation=(X=0,Y=0,Z=0),Scale3D=(X=1,Y=1,Z=1))))
    Weight=1
    WeightMinimum=0
    ItemToModify.0.Transform.Rotation=&RigVMFunction_MathQuaternionFromRotatorV2(Value=@Make.Struct) {template="Cast::Execute(in Value,out Result)"}.Result
    ExecutePin=@n6.False
   )
   RigVMFunction_MathBoolToFloat = &RigVMFunction_MathBoolToFloat(Value=@n4.Value)
   AlphaInterp = &RigVMFunction_AlphaInterp(InterpSpeedDecreasing=15, InterpSpeedIncreasing=15, Scale=1, bClampResult=False, Value=@RigVMFunction_MathBoolToFloat.Result)
   &RigUnit_ModifyTransforms(
    ItemToModify="((Item=(Type=Bone,Name=\"dyn_bagDrive_master\"),Transform=(Rotation=(X=0,Y=0,Z=-0.3079,W=0.9514),Translation=(X=0,Y=0,Z=0),Scale3D=(X=1,Y=1,Z=1))),(Item=(Type=Bone,Name=\"dyn_bag_00\"),Transform=(Rotation=(X=0,Y=0,Z=0.1951,W=0.9808),Translation=(X=0,Y=0,Z=0),S ...(truncated)"
    WeightMinimum=0
    Weight=@AlphaInterp.Result
    ExecutePin=@RigVMFunction_ControlFlowBranch.True
   )
   n9 = &RigUnit_SphericalPoseReader(
    ActiveRegionScaleFactors=(PositiveWidth=3.5,NegativeWidth=0.5,PositiveHeight=3,NegativeHeight=3)
    ActiveRegionSize=0.1
    Debug=(bDrawDebug=false,DebugScale=25,DebugSegments=20,DebugThickness=0.25)
    DriverAxis=(X=-1,Y=0,Z=0)
    DriverItem=(Type=Bone,Name="thigh_l")
    FalloffRegionScaleFactors=(PositiveWidth=1,NegativeWidth=1.6,PositiveHeight=2.5,NegativeHeight=1)
    FalloffSize=0.45
    OptionalParentItem=(Type=Bone)
    RotationOffset=(X=90,Y=0,Z=-35)
    ExecutePin=@RigVMFunction_Sequence_3.B
   )
   n10 = &RigVMFunction_AnimEvalRichCurve(
    Curve="(EditorCurveData=(Keys=((InterpMode=RCIM_Cubic,TangentMode=RCTM_User,ArriveTangent=1.3862,LeaveTangent=1.3862),(InterpMode=RCIM_Cubic,TangentMode=RCTM_User,Time=0.6138,Value=0.8703,ArriveTangent=1.0646,LeaveTangent=1.0646),(InterpMode=RCIM_Cubic,TangentMod ...(truncated)"
    SourceMinimum=0
    TargetMinimum=0
    Value=@n9.OutputParam
   )
   > &RigUnit_ModifyTransforms(
    ItemToModify="((Item=(Type=Bone,Name=\"dyn_beltFront_anchor_a\"),Transform=(Rotation=(X=0,Y=0,Z=0.1467,W=0.9892),Translation=(X=2.4769,Y=-0.5946,Z=2.7732),Scale3D=(X=1,Y=1,Z=1))),(Item=(Type=Bone,Name=\"dyn_beltFront_anchor_b\"),Transform=(Rotation=(X=0,Y=0,Z=0,W=1),Transla ...(truncated)"
    WeightMinimum=0
    Weight=@n10.Result
   )
   > n11 = &RigUnit_SphericalPoseReader(
    ActiveRegionScaleFactors=(PositiveWidth=3.5,NegativeWidth=0.5,PositiveHeight=3,NegativeHeight=3)
    ActiveRegionSize=0.1
    Debug=(bDrawDebug=False,DebugScale=25,DebugSegments=20,DebugThickness=0.25)
    DriverAxis=(X=1,Y=0,Z=0)
    DriverItem=(Type=Bone,Name="thigh_r")
    FalloffRegionScaleFactors=(PositiveWidth=1,NegativeWidth=1.6,PositiveHeight=2.5,NegativeHeight=1)
    FalloffSize=0.45
    FlipHeightScaling=false
    FlipWidthScaling=true
    OptionalParentItem=(Type=Bone)
    RotationOffset=(X=-90,Y=0,Z=-35)
   )
   > &RigUnit_ModifyTransforms(
    ItemToModify="((Item=(Type=Bone,Name=\"dyn_beltFront_anchor_a\"),Transform=(Rotation=(X=0,Y=0,Z=0,W=1),Translation=(X=0,Y=0,Z=0),Scale3D=(X=1,Y=1,Z=1))),(Item=(Type=Bone,Name=\"dyn_beltFront_anchor_b\"),Transform=(Rotation=(X=0,Y=0,Z=0,W=1),Translation=(X=0,Y=0,Z=0),Scale3D ...(truncated)"
    WeightMinimum=0
    Weight=@n11.OutputParam
   )
   n12 = &RigUnit_SphericalPoseReader(
    ActiveRegionScaleFactors=(PositiveWidth=3,NegativeWidth=1.2,PositiveHeight=3,NegativeHeight=3)
    ActiveRegionSize=0.1
    Debug=(bDrawDebug=False,DebugScale=25,DebugSegments=20,DebugThickness=0.25)
    DriverAxis=(X=1,Y=0,Z=0)
    DriverItem=(Type=Bone,Name="thigh_r")
    FalloffRegionScaleFactors=(PositiveWidth=0.1,NegativeWidth=1.1,PositiveHeight=1,NegativeHeight=1)
    FalloffSize=0.45
    FlipHeightScaling=false
    FlipWidthScaling=true
    OptionalParentItem=(Type=Bone)
    RotationOffset=(X=-90,Y=0,Z=0)
    ExecutePin=@RigVMFunction_Sequence_3.C
   )
   > &RigUnit_ModifyTransforms(
    ItemToModify="((Item=(Type=Bone,Name=\"dyn_shuriken_a_01\"),Transform=(Rotation=(X=0,Y=0,Z=0,W=1),Translation=(X=0,Y=0,Z=0),Scale3D=(X=1,Y=1,Z=0.8))),(Item=(Type=Bone,Name=\"dyn_shuriken_b_01\"),Transform=(Rotation=(X=0,Y=0,Z=0,W=1),Translation=(X=0,Y=0,Z=0),Scale3D=(X=1,Y= ...(truncated)"
    WeightMinimum=0
    Weight=@n12.OutputParam
   )
   > n13 = &RigUnit_SphericalPoseReader(
    ActiveRegionScaleFactors=(PositiveWidth=3,NegativeWidth=1.2,PositiveHeight=3,NegativeHeight=3)
    ActiveRegionSize=0.1
    Debug=(bDrawDebug=false,DebugScale=25,DebugSegments=20,DebugThickness=0.25)
    DriverAxis=(X=1,Y=0,Z=0)
    DriverItem=(Type=Bone,Name="thigh_l")
    FalloffRegionScaleFactors=(PositiveWidth=0.1,NegativeWidth=1.1,PositiveHeight=1,NegativeHeight=1)
    FalloffSize=0.45
    FlipHeightScaling=false
    FlipWidthScaling=true
    OptionalParentItem=(Type=Bone)
    RotationOffset=(X=-90,Y=0,Z=0)
   )
   > &RigUnit_ModifyTransforms(
    ItemToModify="((Item=(Type=Bone,Name=\"dyn_beltFront_anchor_b_master\"),Transform=(Rotation=(X=-0.1467,Y=0,Z=0,W=0.9892),Translation=(X=0.9736,Y=-3.5697,Z=6.4543),Scale3D=(X=1,Y=1,Z=1))),(Item=(Type=Bone,Name=\"dyn_beltFront_anchor_c_master\"),Transform=(Rotation=(X=0,Y=0,Z ...(truncated)"
    WeightMinimum=0
    Weight=@n13.OutputParam
   )
   > n14 = &RigUnit_SphericalPoseReader(
    ActiveRegionScaleFactors=(PositiveWidth=3,NegativeWidth=2.2,PositiveHeight=3,NegativeHeight=3)
    ActiveRegionSize=0.1
    Debug=(bDrawDebug=true,DebugScale=25,DebugSegments=20,DebugThickness=0.25)
    DriverAxis=(X=1,Y=0,Z=0)
    DriverItem=(Type=Bone,Name="thigh_r")
    FalloffRegionScaleFactors=(PositiveWidth=0.1,NegativeWidth=1.1,PositiveHeight=1,NegativeHeight=1)
    FalloffSize=0.45
    FlipHeightScaling=false
    FlipWidthScaling=true
    OptionalParentItem=(Type=Bone)
    RotationOffset=(X=-17.4831,Y=3.1705,Z=42.1829)
   )
   Remap = &RigVMFunction_MathDoubleRemap(
    SourceMaximum=44
    SourceMinimum=60
    TargetMaximum=1
    TargetMinimum=0
    Value=&RigUnit_GetRelativeTransformForItem(Child=(Type=Bone,Name="pelvis"), Parent=(Type=Bone,Name="root"), bChildInitial=False, bParentInitial=False).RelativeTransform.Translation.Z
   )
   Not_Equals = RigVMDispatchNode(B=1, A=RigVMVariableNode() {variable=DBNO Alpha}.Value) {template="DISPATCH_RigVMDispatch_CoreNotEquals(in A,in B,out Result)"}
   Multiply = RigVMAggregateNode(B=@Remap.Result, C=&RigVMFunction_MathBoolToFloat(Value=@Not_Equals.Result).Result, A=@n14.OutputParam)
   > &RigUnit_ModifyTransforms(
    ItemToModify="((Item=(Type=Bone,Name=\"dyn_bagDrive_master\"),Transform=(Rotation=(X=0,Y=0,Z=0,W=1),Translation=(X=-0.968,Y=-0.2032,Z=8.9893),Scale3D=(X=1,Y=1,Z=1))),(Item=(Type=Bone,Name=\"dyn_bag_00\"),Transform=(Rotation=(X=0,Y=0,Z=0,W=1),Translation=(X=0,Y=0,Z=0),Scale3 ...(truncated)"
    WeightMinimum=0
    Weight=@Multiply.Result
   )
   n15 = &RigUnit_SphericalPoseReader(
    ActiveRegionScaleFactors=(PositiveWidth=1,NegativeWidth=1,PositiveHeight=1,NegativeHeight=1)
    ActiveRegionSize=0.1
    Debug=(bDrawDebug=false,DebugScale=25,DebugSegments=20,DebugThickness=0.25)
    DriverAxis=(X=1,Y=0,Z=0)
    DriverItem=(Type=Bone,Name="thigh_r")
    FalloffRegionScaleFactors=(PositiveWidth=1,NegativeWidth=1,PositiveHeight=1,NegativeHeight=1)
    FalloffSize=0.2
    FlipHeightScaling=false
    FlipWidthScaling=true
    OptionalParentItem=(Type=Bone)
    RotationOffset=(X=-18.789,Y=77.2214,Z=45.356)
    ExecutePin=@RigVMFunction_Sequence_3.D
   )
   > &RigUnit_ModifyTransforms(
    ItemToModify="((Item=(Type=Bone,Name=\"dyn_beltFront_anchor_b\"),Transform=(Rotation=(X=-0.1804,Y=-0.3013,Z=-0.0582,W=0.9345),Translation=(X=-0.7579,Y=-0.6675,Z=2.5679),Scale3D=(X=1,Y=1,Z=1))),(Item=(Type=Bone,Name=\"dyn_beltFront_anchor_c\"),Transform=(Rotation=(X=0,Y=-0.0 ...(truncated)"
    WeightMinimum=0
    Weight=@n15.OutputParam
   )
   n16 = &RigVMFunction_Sequence(ExecuteContext=@RigVMFunction_Sequence_1.D)
   n17 = &RigUnit_SphericalPoseReader(
    ActiveRegionScaleFactors=(PositiveWidth=1,NegativeWidth=1,PositiveHeight=1,NegativeHeight=1)
    ActiveRegionSize=0.1
    Debug=(bDrawDebug=False,bDraw2D=False,bDrawLocalAxes=False,DebugScale=25,DebugSegments=20,DebugThickness=0.25)
    DriverAxis=(X=0,Y=1,Z=0)
    DriverItem=(Type=Bone,Name="foot_l")
    FalloffRegionScaleFactors=(PositiveWidth=1,NegativeWidth=0.1,PositiveHeight=0.5,NegativeHeight=0.5)
    FalloffSize=0.33
    FlipHeightScaling=False
    FlipWidthScaling=False
    OptionalParentItem=(Type=Bone,Name="None")
    RotationOffset=(X=90,Y=0,Z=60)
    ExecutePin=@n16.A
   )
   > &RigUnit_ModifyTransforms(
    ItemToModify="((Item=(Type=Bone,Name=\"deform_foot_in_l\"),Transform=(Rotation=(X=0.4276,Y=0,Z=0,W=0.904),Translation=(X=0.0002,Y=-15.7196,Z=-5.5454),Scale3D=(X=1,Y=1,Z=1))),(Item=(Type=Bone,Name=\"deform_foot_out_l\"),Transform=(Rotation=(X=0.2667,Y=0,Z=0,W=0.9638),Transla ...(truncated)"
    WeightMinimum=0
    Weight=@n17.OutputParam
   )
   > n18 = &RigUnit_SphericalPoseReader(
    ActiveRegionScaleFactors=(PositiveWidth=1,NegativeWidth=1,PositiveHeight=1,NegativeHeight=1)
    ActiveRegionSize=0.1
    Debug=(bDrawDebug=False,bDraw2D=False,bDrawLocalAxes=False,DebugScale=25,DebugSegments=20,DebugThickness=0.25)
    DriverAxis=(X=0,Y=1,Z=0)
    DriverItem=(Type=Bone,Name="foot_l")
    FalloffRegionScaleFactors=(PositiveWidth=1,NegativeWidth=0.1,PositiveHeight=0.5,NegativeHeight=0.5)
    FalloffSize=0.33
    FlipHeightScaling=False
    FlipWidthScaling=False
    OptionalParentItem=(Type=Bone,Name="None")
    RotationOffset=(X=87.9064,Y=-180.3168,Z=-61.8994)
   )
   > &RigUnit_ModifyTransforms(
    ItemToModify="((Item=(Type=Bone,Name=\"deform_foot_in_l\"),Transform=(Rotation=(X=-0.2191,Y=0,Z=0,W=0.9757),Translation=(X=0,Y=0.4197,Z=-0.1985),Scale3D=(X=1,Y=1,Z=1))),(Item=(Type=Bone,Name=\"deform_foot_out_l\"),Transform=(Rotation=(X=-0.2903,Y=0,Z=0,W=0.9569),Translation ...(truncated)"
    WeightMinimum=0
    Weight=@n18.OutputParam
   )
   n19 = &RigUnit_SphericalPoseReader(
    ActiveRegionScaleFactors=(PositiveWidth=1,NegativeWidth=1,PositiveHeight=1,NegativeHeight=1)
    ActiveRegionSize=0.1
    Debug=(bDrawDebug=False,bDraw2D=False,bDrawLocalAxes=False,DebugScale=25,DebugSegments=20,DebugThickness=0.25)
    DriverAxis=(X=0,Y=1,Z=0)
    DriverItem=(Type=Bone,Name="foot_r")
    FalloffRegionScaleFactors=(PositiveWidth=1,NegativeWidth=0.1,PositiveHeight=0.5,NegativeHeight=0.5)
    FalloffSize=0.33
    FlipHeightScaling=False
    FlipWidthScaling=False
    OptionalParentItem=(Type=Bone,Name="None")
    RotationOffset=(X=90,Y=0,Z=60)
    ExecutePin=@n16.B
   )
   > &RigUnit_ModifyTransforms(
    ItemToModify="((Item=(Type=Bone,Name=\"deform_foot_in_r\"),Transform=(Rotation=(X=0.4276,Y=0,Z=0,W=0.904),Translation=(X=0.0002,Y=15.7196,Z=5.5454),Scale3D=(X=1,Y=1,Z=1))),(Item=(Type=Bone,Name=\"deform_foot_out_r\"),Transform=(Rotation=(X=0.2667,Y=0,Z=0,W=0.9638),Translati ...(truncated)"
    WeightMinimum=0
    Weight=@n19.OutputParam
   )
   > n20 = &RigUnit_SphericalPoseReader(
    ActiveRegionScaleFactors=(PositiveWidth=1,NegativeWidth=1,PositiveHeight=1,NegativeHeight=1)
    ActiveRegionSize=0.1
    Debug=(bDrawDebug=False,bDraw2D=False,bDrawLocalAxes=False,DebugScale=25,DebugSegments=20,DebugThickness=0.25)
    DriverAxis=(X=0,Y=1,Z=0)
    DriverItem=(Type=Bone,Name="foot_r")
    FalloffRegionScaleFactors=(PositiveWidth=1,NegativeWidth=0.1,PositiveHeight=0.5,NegativeHeight=0.5)
    FalloffSize=0.33
    FlipHeightScaling=False
    FlipWidthScaling=False
    OptionalParentItem=(Type=Bone,Name="None")
    RotationOffset=(X=87.9064,Y=-180.3168,Z=-61.8994)
   )
   > &RigUnit_ModifyTransforms(
    ItemToModify="((Item=(Type=Bone,Name=\"deform_foot_in_r\"),Transform=(Rotation=(X=-0.2191,Y=0,Z=0,W=0.9757),Translation=(X=0,Y=-0.4197,Z=0.1985),Scale3D=(X=1,Y=1,Z=1))),(Item=(Type=Bone,Name=\"deform_foot_out_r\"),Transform=(Rotation=(X=-0.2903,Y=0,Z=0,W=0.9569),Translation ...(truncated)"
    WeightMinimum=0
    Weight=@n20.OutputParam
   )
   n21 = &RigUnit_SphericalPoseReader(
    ActiveRegionScaleFactors=(PositiveWidth=1,NegativeWidth=1,PositiveHeight=1,NegativeHeight=1)
    ActiveRegionSize=0.1
    Debug=(bDrawDebug=True,bDraw2D=False,bDrawLocalAxes=False,DebugScale=25,DebugSegments=20,DebugThickness=0.25)
    DriverAxis=(X=1,Y=0,Z=0)
    DriverItem=(Type=Bone,Name="upperarm_r")
    FalloffRegionScaleFactors=(PositiveWidth=1,NegativeWidth=1,PositiveHeight=1,NegativeHeight=1)
    FalloffSize=0.2
    FlipHeightScaling=False
    FlipWidthScaling=False
    OptionalParentItem=(Type=Bone,Name="spine_05")
    RotationOffset=(X=98.6622,Y=7.3801,Z=43.0255)
    ExecutePin=@RigVMFunction_Sequence_1.G
   )
   > &RigUnit_ModifyTransforms(
    ItemToModify=((Item=(Type=Bone,Name="deform_upperarm_up_r"),Transform=(Rotation=(X=0.0684,Y=0.0159,Z=-0.2259,W=0.9716),Translation=(X=-5.5966,Y=-18.8028,Z=-1.1529),Scale3D=(X=1.6,Y=3.201,Z=1))))
    WeightMinimum=0
    Weight=@n21.OutputParam
   )
   > n22 = &RigUnit_SphericalPoseReader(
    ActiveRegionScaleFactors=(PositiveWidth=1,NegativeWidth=1,PositiveHeight=1,NegativeHeight=1)
    ActiveRegionSize=0.1
    Debug=(bDrawDebug=True,bDraw2D=False,bDrawLocalAxes=False,DebugScale=25,DebugSegments=20,DebugThickness=0.25)
    DriverAxis=(X=1,Y=0,Z=0)
    DriverItem=(Type=Bone,Name="upperarm_l")
    FalloffRegionScaleFactors=(PositiveWidth=1,NegativeWidth=1,PositiveHeight=1,NegativeHeight=1)
    FalloffSize=0.2
    FlipHeightScaling=False
    FlipWidthScaling=False
    OptionalParentItem=(Type=Bone,Name="spine_05")
    RotationOffset=(X=42.8957,Y=27.6462,Z=-14.0751)
   )
   > &RigUnit_ModifyTransforms(
    ItemToModify=((Item=(Type=Bone,Name="deform_upperarm_up_l"),Transform=(Rotation=(X=-0.0684,Y=-0.0159,Z=-0.2259,W=0.9716),Translation=(X=5.5966,Y=18.8028,Z=-1.1529),Scale3D=(X=1.6,Y=3.201,Z=1))))
    WeightMinimum=0
    Weight=@n22.OutputParam
   )
   Get Delta Ref = RigVMFunctionReferenceNode(Item=(Type=Bone,Name="pelvis"), ExecuteContext=@RigVMFunction_Sequence_1.A)
   > RigVMVariableNode(Value=/Script/RigVM.RigVMFunction_MathDoubleGreater(
     B=0
     A=&RigVMFunction_MathDoubleRemap(SourceMaximum=90, SourceMinimum=80, TargetMaximum=1, TargetMinimum=0, Value=@"Get Delta Ref".Result.Rotation.Yaw).Result
    ) {is=struct RigVMUnitNode, template="Greater::Execute(in A,in B,out Result)"}.Result) {variable=Is pot SkydiveFloating}
   > Get Delta Ref_1 = RigVMFunctionReferenceNode(Item=(Type=Bone,Name="pelvis"))
   n23 = RigVMDispatchNode(Struct=&RigVMFunction_MathQuaternionToRotator(Value=&RigUnit_GetRelativeTransformForItem(Child=(Type=Bone,Name="pelvis"), Parent=(Type=Bone,Name="root"), bChildInitial=False, bParentInitial=false).RelativeTransform.Rotation) {template="Cast::Execute(in Value,out Result)"}.Result) {template="DISPATCH_RigVMDispatch_BreakStruct(in Struct,out Elements)"}
   Greater = &RigVMFunction_MathDoubleGreater(B=0, A=@n23.Elements.Pitch) {template="Greater::Execute(in A,in B,out Result)"}
   n24 = /Script/RigVM.RigVMFunction_MathDoubleLess(B=500, A=RigVMVariableNode() {variable=Abs Velocity}.Value) {is=struct RigVMUnitNode, template="Less::Execute(in A,in B,out Result)"}
   RigVMFunction_MathBoolAnd = RigVMAggregateNode(
    A=@Greater.Result
    B=&RigVMFunction_MathDoubleLess(B=28, A=@n23.Elements.Pitch) {template="Less::Execute(in A,in B,out Result)"}.Result
    C=@n24.Result
   )
   > RigVMVariableNode(Value=@RigVMFunction_MathBoolAnd.Result) {variable=Is pot DBNO}
   Multiply_4 = RigVMAggregateNode(
    B=&RigVMFunction_MathDoubleRemap(
     SourceMaximum=45
     SourceMinimum=55
     TargetMaximum=1
     TargetMinimum=0
     Value=&RigUnit_GetRelativeTransformForItem(Child=(Type=Bone,Name="pelvis"), Parent=(Type=Bone,Name="root"), bChildInitial=False, bParentInitial=False).RelativeTransform.Translation.Z
    ).Result
    A=&RigVMFunction_MathDoubleRemap(SourceMaximum=50, SourceMinimum=40, TargetMaximum=1, TargetMinimum=0, Value=@"Get Delta Ref_1".Result.Rotation.Yaw).Result
    C=&RigVMFunction_MathBoolToFloat(Value=@n24.Result).Result
   )
   > RigVMVariableNode(Value=@Multiply_4.Result) {variable=DBNO Alpha}
   n25 = &RigUnit_GetRelativeTransformForItem(Child=(Type=Bone,Name="pelvis"), Parent=(Type=Bone,Name="root"), bChildInitial=False, bParentInitial=False)
   n26 = &RigVMFunction_MathDoubleMul(
    A=&RigVMFunction_MathDoubleRemap(SourceMaximum=50, SourceMinimum=60, TargetMaximum=1, TargetMinimum=0, Value=@n25.RelativeTransform.Translation.Z).Result
    B=&RigVMFunction_MathDoubleRemap(SourceMaximum=35, SourceMinimum=25, TargetMaximum=1, TargetMinimum=0, Value=@n25.RelativeTransform.Translation.Z).Result
   )
   n27 = &RigVMFunction_MathDoubleMul(
    B=@n26.Result
    A=&RigVMFunction_MathDoubleRemap(SourceMaximum=1, SourceMinimum=0, TargetMaximum=0, TargetMinimum=1, Value=RigVMVariableNode() {variable=DBNO Alpha}.Value).Result
   )
   &RigUnit_ModifyTransforms(
    ItemToModify=((Item=(Type=Bone,Name="dyn_bagDrive_master"),Transform=(Rotation=(X=0.8306,Y=0.3441,Z=0.1675,W=0.4045),Translation=(X=0,Y=-2,Z=0),Scale3D=(X=1,Y=1,Z=1))))
    WeightMinimum=0
    Weight=@n27.Result
    ExecutePin=@RigVMFunction_Sequence_1.E
   )
   > /Script/ControlRig.RigUnit_SetCurveValue(Curve=CrouchIdleAlpha, Value=@n27.Result) {is=struct RigVMUnitNode}
   n28 = &RigVMFunction_Sequence(ExecuteContext=@RigVMFunction_Sequence_1.F)
   DeformRig_GetDeltaRefTransform_FNC_2 = RigVMFunctionReferenceNode(Item=(Type=Bone,Name="calf_l"), ExecuteContext=@n28.A)
   > &RigUnit_ModifyTransforms(
    ItemToModify="((Item=(Type=Bone,Name=\"deform_knee_inner_l\"),Transform=(Rotation=(X=0,Y=0,Z=-0.3607,W=0.9327),Translation=(X=-15.6402,Y=-9.9943,Z=0.0001),Scale3D=(X=1,Y=1,Z=1))),(Item=(Type=Bone,Name=\"deform_knee_outer_l\"),Transform=(Rotation=(X=0,Y=0,Z=-0.3978,W=0.9175) ...(truncated)"
    WeightMinimum=0
    Weight=&RigVMFunction_MathFloatRemap(SourceMaximum=110, SourceMinimum=10, bClamp=true, Value=@DeformRig_GetDeltaRefTransform_FNC_2.OutTransform.Rotation.Yaw).Result
   )
   > &RigUnit_ModifyTransforms(
    ItemToModify="((Item=(Type=Bone,Name=\"deform_knee_inner_l\"),Transform=(Rotation=(X=0,Y=0,Z=0,W=1),Translation=(X=-6.6497,Y=0,Z=-2.962),Scale3D=(X=1,Y=1,Z=1))),(Item=(Type=Bone,Name=\"deform_knee_outer_l\"),Transform=(Rotation=(X=0,Y=0,Z=0,W=1),Translation=(X=-9.1262,Y=-6. ...(truncated)"
    WeightMinimum=0
    Weight=&RigVMFunction_MathDoubleMul(
     A=&RigVMFunction_MathDoubleRemap(
      SourceMaximum=136
      SourceMinimum=110
      TargetMaximum=1
      TargetMinimum=0
      Value=@DeformRig_GetDeltaRefTransform_FNC_2.OutTransform.Rotation.Yaw
     ).Result
     B=@n26.Result
    ).Result
   )
   DeformRig_GetDeltaRefTransform_FNC_1_1 = RigVMFunctionReferenceNode(Item=(Type=Bone,Name="calf_r"), ExecuteContext=@n28.B)
   > &RigUnit_ModifyTransforms(
    ItemToModify="((Item=(Type=Bone,Name=\"deform_knee_inner_r\"),Transform=(Rotation=(X=0,Y=0,Z=-0.3607,W=0.9327),Translation=(X=15.6402,Y=9.9943,Z=0.0001),Scale3D=(X=1,Y=1,Z=1))),(Item=(Type=Bone,Name=\"deform_knee_outer_r\"),Transform=(Rotation=(X=0,Y=0,Z=-0.3978,W=0.9175),T ...(truncated)"
    WeightMinimum=0
    Weight=&RigVMFunction_MathFloatRemap(SourceMaximum=110, SourceMinimum=10, bClamp=true, Value=@DeformRig_GetDeltaRefTransform_FNC_1_1.OutTransform.Rotation.Yaw).Result
   )
   > &RigUnit_ModifyTransforms(
    ItemToModify="((Item=(Type=Bone,Name=\"deform_knee_inner_r\"),Transform=(Rotation=(X=0,Y=0,Z=0,W=1),Translation=(X=6.6497,Y=0,Z=-2.962),Scale3D=(X=1,Y=1,Z=1))),(Item=(Type=Bone,Name=\"deform_knee_outer_r\"),Transform=(Rotation=(X=0,Y=0,Z=0,W=1),Translation=(X=9.1262,Y=6.320 ...(truncated)"
    WeightMinimum=0
    Weight=&RigVMFunction_MathDoubleMul(
     A=&RigVMFunction_MathDoubleRemap(
      SourceMaximum=136
      SourceMinimum=110
      TargetMaximum=1
      TargetMinimum=0
      Value=@DeformRig_GetDeltaRefTransform_FNC_1_1.OutTransform.Rotation.Yaw
     ).Result
     B=@n26.Result
    ).Result
   )
  ]}
]
hierarchy [
 root {loc=<0 0 0>, rot=<0 0 0 1>}
 pelvis {parent=root, loc=<0 -0.7071 93.1191>, rot=<0.0127 0.707 -0.0127 -0.707>}
 spine_01 {parent=pelvis, loc=<6.7494 -0.6388 0>, rot=<0 0 0.034 -0.9994>}
 spine_02 {parent=spine_01, loc=<7.252 -0.035 0>, rot=<0 0 0.0238 -0.9997>}
 spine_03 {parent=spine_02, loc=<6.0106 0 0>, rot=<0 0 0.0149 -0.9999>}
 spine_04 {parent=spine_03, loc=<8.1023 0 0>, rot=<0 0 -0.0132 -0.9999>}
 spine_05 {parent=spine_04, loc=<10.0142 0 0>, rot=<0 0 -0.0015 -1>}
 attach_backpack {parent=spine_05, loc=<-0.123 -13.473 0>, rot=<-0.0093 -0.707 0.0093 -0.707>}
 neck_01 {parent=spine_05, loc=<11.3138 0.1529 0>, rot=<0 0 -0.2003 -0.9797>}
 neck_02 {parent=neck_01, loc=<4.5627 0 0>, rot=<0 0 0.0504 -0.9987>}
 head {parent=neck_02, loc=<5.3914 0 0>, rot=<0 0 0.115 -0.9934>}
 clavicle_l {parent=spine_05, loc=<4.3575 -0.1466 -2.3223>, rot=<0.0465 -0.7289 -0.009 -0.6829>}
 upperarm_l {parent=clavicle_l, loc=<15.5111 0 0>, rot=<0.0226 -0.3234 -0.0088 -0.9459>}
 lowerarm_l {parent=upperarm_l, loc=<26.6514 0 0>, rot=<0.0391 -0.0099 -0.2453 -0.9686>}
 hand_l {parent=lowerarm_l, loc=<22.3491 0 0>, rot=<-0.7081 -0.0507 0.0975 -0.6975>}
 thumb_01_l {parent=hand_l, loc=<1.3118 -0.3191 -2.459>, rot=<0.6546 -0.4042 0.0664 -0.6354>}
 thumb_02_l {parent=thumb_01_l, loc=<5.9336 -0.0448 0.0401>, rot=<-0.0037 0.0213 0.298 -0.9543>}
 thumb_03_l {parent=thumb_02_l, loc=<2.2223 0.3682 -0.0907>, rot=<0.0118 -0.013 0.0066 -0.9998>}
 index_metacarpal_l {parent=hand_l, loc=<1.6535 0.0183 -2.96>, rot=<0.0158 -0.0124 -0.0173 -0.9996>}
 index_01_l {parent=index_metacarpal_l, loc=<8.3008 0.0357 -0.0524>, rot=<-0.0044 0.0485 0.0924 -0.9945>}
 index_02_l {parent=index_01_l, loc=<3.3587 -0.3004 0.0452>, rot=<-0.002 0.0027 0.3225 -0.9466>}
 index_03_l {parent=index_02_l, loc=<2.5578 -0.123 -0.0252>, rot=<-0.002 0.0067 0.2558 -0.9667>}
 middle_metacarpal_l {parent=hand_l, loc=<2.1856 -0.1725 -0.5984>, rot=<0.006 0.0168 -0.0329 -0.9993>}
 middle_01_l {parent=middle_metacarpal_l, loc=<7.3275 -0.2447 0.0129>, rot=<-0.0208 0.0042 0.0858 -0.9961>}
 middle_02_l {parent=middle_01_l, loc=<4.0123 0.1401 0.1175>, rot=<-0.0218 0.0258 0.3512 -0.9357>}
 middle_03_l {parent=middle_02_l, loc=<2.6077 0.0534 0.1373>, rot=<0.0082 0.0188 0.1708 -0.9851>}
 ring_metacarpal_l {parent=hand_l, loc=<2.2074 -0.3625 1.3638>, rot=<-0.0284 0.0722 -0.0458 -0.9959>}
 ring_01_l {parent=ring_metacarpal_l, loc=<7.0295 -0.6197 -0.1446>, rot=<-0.0489 -0.0252 0.1892 -0.9804>}
 ring_02_l {parent=ring_01_l, loc=<3.6174 0.4647 -0.1027>, rot=<0 0 0.2806 -0.9598>}
 ring_03_l {parent=ring_02_l, loc=<2.4152 0.114 0.0199>, rot=<0 0 0.1189 -0.9929>}
 pinky_metacarpal_l {parent=hand_l, loc=<1.7616 -0.9753 3.3814>, rot=<-0.0397 0.0814 -0.069 -0.9935>}
 pinky_01_l {parent=pinky_metacarpal_l, loc=<7.1709 -0.4831 -0.1094>, rot=<-0.0654 -0.0035 0.2242 -0.9723>}
 pinky_02_l {parent=pinky_01_l, loc=<2.8274 0.0893 -0.1043>, rot=<0.0037 0.0011 0.2854 -0.9584>}
 pinky_03_l {parent=pinky_02_l, loc=<2.2046 0.0307 -0.0078>, rot=<-0.0002 0.0098 0.1445 -0.9895>}
 deform_wrist_dn_base_l {parent=hand_l, loc=<0 -2.2342 0>, rot=<0 0 0 -1>}
 deform_wrist_dn_l {parent=deform_wrist_dn_base_l, loc=<0 0 0>, rot=<0 0 0 1>}
 deform_wrist_up_base_l {parent=hand_l, loc=<0 2.2682 0>, rot=<0 0 0 -1>}
 deform_wrist_up_l {parent=deform_wrist_up_base_l, loc=<0 0 0>, rot=<0 0 0 1>}
 hand_attach_l {parent=hand_l, loc=<8.5 -3 0>, rot=<0.7071 0.7071 0 0>}
 weapon_l {parent=hand_l, loc=<-0.1974 -3.4111 0.3811>, rot=<0 0 0 1>}
 lowerarm_twist_02_l {parent=lowerarm_l, loc=<9.6199 0 0>, rot=<0 0 0 -1>}
 lowerarm_twist_01_l {parent=lowerarm_l, loc=<17.8549 0 0>, rot=<0 0 0 -1>}
 deform_elbow_in_base_l {parent=lowerarm_l, loc=<0.2112 5.259 0.8274>, rot=<0 0 0 -1>}
 deform_elbow_in_l {parent=deform_elbow_in_base_l, loc=<0 0 0>, rot=<0 0 0 1>}
 deform_elbow_out_base_l {parent=lowerarm_l, loc=<0 -4 0>, rot=<0 0 0 -1>}
 deform_elbow_out_l {parent=deform_elbow_out_base_l, loc=<0 0 0>, rot=<0 0 0 1>}
 upperarm_twist_01_l {parent=upperarm_l, loc=<0.0041 0 0>, rot=<0 0 0 -1>}
 deform_upperarm_bk_base_l {parent=upperarm_twist_01_l, loc=<1.0813 -7.2193 -1.9453>, rot=<0 0 0 1>}
 deform_upperarm_bk_l {parent=deform_upperarm_bk_base_l, loc=<0 0 0>, rot=<0 0 0 1>}
 deform_upperarm_fr_base_l {parent=upperarm_twist_01_l, loc=<-0.3691 7.0506 -2.765>, rot=<0 0 0 1>}
 deform_upperarm_fr_l {parent=deform_upperarm_fr_base_l, loc=<0 0 0>, rot=<0 0 0 1>}
 deform_upperarm_up_base_l {parent=upperarm_twist_01_l, loc=<-3.2216 -1.1391 3.3311>, rot=<0 0 0 1>}
 deform_upperarm_up_l {parent=deform_upperarm_up_base_l, loc=<0 0 0>, rot=<0 0 0 1>}
 upperarm_twist_02_l {parent=upperarm_l, loc=<13.7941 0 0>, rot=<0 0 0 -1>}
 attach_cape {parent=spine_05, loc=<9.017 -11.714 0>, rot=<0.0616 -0.7044 -0.0616 -0.7044>}
 clavicle_r {parent=spine_05, loc=<4.3575 -0.1466 2.3223>, rot=<0.7289 0.0465 -0.6829 0.009>}
 upperarm_r {parent=clavicle_r, loc=<-15.5111 0 0>, rot=<0.0226 -0.3234 -0.0088 -0.9459>}
 lowerarm_r {parent=upperarm_r, loc=<-26.6514 0 0>, rot=<0.0391 -0.0099 -0.2453 -0.9686>}
 hand_r {parent=lowerarm_r, loc=<-22.3491 0 0>, rot=<-0.7081 -0.0507 0.0975 -0.6975>}
 thumb_01_r {parent=hand_r, loc=<-1.3118 0.3191 2.459>, rot=<0.6546 -0.4042 0.0664 -0.6354>}
 thumb_02_r {parent=thumb_01_r, loc=<-5.9336 0.0448 -0.0401>, rot=<-0.0037 0.0213 0.298 -0.9543>}
 thumb_03_r {parent=thumb_02_r, loc=<-2.2223 -0.3682 0.0907>, rot=<0.0118 -0.013 0.0066 -0.9998>}
 index_metacarpal_r {parent=hand_r, loc=<-1.6535 -0.0183 2.96>, rot=<0.0158 -0.0124 -0.0173 -0.9996>}
 index_01_r {parent=index_metacarpal_r, loc=<-8.3008 -0.0357 0.0524>, rot=<-0.0044 0.0485 0.0924 -0.9945>}
 index_02_r {parent=index_01_r, loc=<-3.3587 0.3004 -0.0452>, rot=<-0.002 0.0027 0.3225 -0.9466>}
 index_03_r {parent=index_02_r, loc=<-2.5578 0.123 0.0252>, rot=<-0.002 0.0067 0.2558 -0.9667>}
 middle_metacarpal_r {parent=hand_r, loc=<-2.1856 0.1725 0.5984>, rot=<0.006 0.0168 -0.0329 -0.9993>}
 middle_01_r {parent=middle_metacarpal_r, loc=<-7.3275 0.2447 -0.0129>, rot=<-0.0208 0.0042 0.0858 -0.9961>}
 middle_02_r {parent=middle_01_r, loc=<-4.0123 -0.1401 -0.1175>, rot=<-0.0218 0.0258 0.3512 -0.9357>}
 middle_03_r {parent=middle_02_r, loc=<-2.6077 -0.0534 -0.1373>, rot=<0.0082 0.0188 0.1708 -0.9851>}
 ring_metacarpal_r {parent=hand_r, loc=<-2.2074 0.3625 -1.3638>, rot=<-0.0284 0.0722 -0.0458 -0.9959>}
 ring_01_r {parent=ring_metacarpal_r, loc=<-7.0295 0.6197 0.1446>, rot=<-0.0489 -0.0252 0.1892 -0.9804>}
 ring_02_r {parent=ring_01_r, loc=<-3.6174 -0.4647 0.1027>, rot=<0 0 0.2806 -0.9598>}
 ring_03_r {parent=ring_02_r, loc=<-2.4152 -0.114 -0.0199>, rot=<0 0 0.1189 -0.9929>}
 pinky_metacarpal_r {parent=hand_r, loc=<-1.7616 0.9753 -3.3814>, rot=<-0.0397 0.0814 -0.069 -0.9935>}
 pinky_01_r {parent=pinky_metacarpal_r, loc=<-7.1709 0.4831 0.1094>, rot=<-0.0654 -0.0035 0.2242 -0.9723>}
 pinky_02_r {parent=pinky_01_r, loc=<-2.8274 -0.0893 0.1043>, rot=<0.0037 0.0011 0.2854 -0.9584>}
 pinky_03_r {parent=pinky_02_r, loc=<-2.2046 -0.0307 0.0078>, rot=<-0.0002 0.0098 0.1445 -0.9895>}
 deform_wrist_dn_base_r {parent=hand_r, loc=<-0.0004 2.2345 -0.0001>, rot=<0 0 0 -1>}
 deform_wrist_dn_r {parent=deform_wrist_dn_base_r, loc=<0 0 0>, rot=<0 0 0 1>}
 deform_wrist_up_base_r {parent=hand_r, loc=<0.0003 -2.2684 0.0001>, rot=<0 0 0 -1>}
 deform_wrist_up_r {parent=deform_wrist_up_base_r, loc=<0 0 0>, rot=<0 0 0 1>}
 weapon_r {parent=hand_r, loc=<0.1974 3.4111 -0.3811>, rot=<0 0 0 1>}
 hand_attach_r {parent=hand_r, loc=<-8.5001 2.9999 0>, rot=<0 0 -0.7071 -0.7071>}
 lowerarm_twist_02_r {parent=lowerarm_r, loc=<-9.6199 0 0>, rot=<0 0 0 -1>}
 lowerarm_twist_01_r {parent=lowerarm_r, loc=<-17.8549 0 0>, rot=<0 0 0 -1>}
 deform_elbow_in_base_r {parent=lowerarm_r, loc=<-0.2111 -5.2591 -0.8275>, rot=<0 0 0 -1>}
 deform_elbow_in_r {parent=deform_elbow_in_base_r, loc=<-0.0001 0 0.0002>, rot=<0 0 0 1>}
 deform_elbow_out_base_r {parent=lowerarm_r, loc=<0 4 -0.0001>, rot=<0 0 0 -1>}
 deform_elbow_out_r {parent=deform_elbow_out_base_r, loc=<0 0 0>, rot=<0 0 0 1>}
 upperarm_twist_01_r {parent=upperarm_r, loc=<-0.0041 0 0>, rot=<0 0 0 -1>}
 deform_upperarm_bk_base_r {parent=upperarm_twist_01_r, loc=<-1.0814 7.2192 1.9455>, rot=<0 0 0 1>}
 deform_upperarm_bk_r {parent=deform_upperarm_bk_base_r, loc=<0 0 0>, rot=<0 0 0 1>}
 deform_upperarm_fr_base_r {parent=upperarm_twist_01_r, loc=<0.3689 -7.0506 2.7652>, rot=<0 0 0 1>}
 deform_upperarm_fr_r {parent=deform_upperarm_fr_base_r, loc=<0 0 0>, rot=<0 0 0 1>}
 deform_upperarm_up_base_r {parent=upperarm_twist_01_r, loc=<3.2214 1.1391 -3.3308>, rot=<0 0 0 1>}
 deform_upperarm_up_r {parent=deform_upperarm_up_base_r, loc=<0 0 0>, rot=<0 0 0 -1>}
 upperarm_twist_02_r {parent=upperarm_r, loc=<-13.7941 0 0>, rot=<0 0 0 -1>}
 deform_pec_aim_base_l {parent=spine_04, loc=<7.8728 12.6388 -2.4123>, rot=<0.0102 -0.4602 0.3979 -0.7936>}
 deform_pec_aim_l {parent=deform_pec_aim_base_l, loc=<0 0 0>, rot=<0 0 0 1>}
 deform_pec_base_l {parent=deform_pec_aim_l, loc=<10.766 0 0>, rot=<0 0 0 1>}
 deform_pec_l {parent=deform_pec_base_l, loc=<0.055 0 0>, rot=<0 0 0 1>}
 deform_pec_aim_base_r {parent=spine_04, loc=<7.8723 12.6388 2.4123>, rot=<-0.4602 -0.0102 0.7936 0.3979>}
 deform_pec_aim_r {parent=deform_pec_aim_base_r, loc=<0 0 0>, rot=<0 0 0 1>}
 deform_pec_base_r {parent=deform_pec_aim_r, loc=<-10.7662 -0.0006 -0.0007>, rot=<0 0 0 1>}
 deform_pec_r {parent=deform_pec_base_r, loc=<0 0 0>, rot=<0 0 0 1>}
 thigh_l {parent=pelvis, loc=<-2.5693 -0.0109 -7.7501>, rot=<0.0012 0.0572 0.0155 -0.9982>}
 calf_l {parent=thigh_l, loc=<-39.0484 0 0>, rot=<0.0006 -0.0161 -0.0391 -0.9991>}
 foot_l {parent=calf_l, loc=<-42.561 0.1077 0.4292>, rot=<0.0418 -0.0455 0.0343 -0.9975>}
 ball_l {parent=foot_l, loc=<-6.1599 13.8264 0.6806>, rot=<0.0001 -0.0001 -0.7063 -0.7079>}
 deform_foot_in_base_l {parent=foot_l, loc=<-0.3388 -3.9009 0.3196>, rot=<-0.0357 -0.7045 -0.0284 -0.7082>}
 deform_foot_in_l {parent=deform_foot_in_base_l, loc=<0 0 0>, rot=<0 0 0 1>}
 deform_foot_out_base_l {parent=foot_l, loc=<1.951 3.8423 0.1665>, rot=<-0.0357 -0.7045 -0.0284 -0.7082>}
 deform_foot_out_l {parent=deform_foot_out_base_l, loc=<0 0 0>, rot=<0 0 0 1>}
 calf_twist_02_l {parent=calf_l, loc=<-18.8064 -0.0967 0.0397>, rot=<0.0237 0.0095 -0.0035 -0.9997>}
 calf_twist_01_l {parent=calf_l, loc=<-35.813 -0.0967 0.0397>, rot=<0.0237 0.0095 -0.0035 -0.9997>}
 deform_knee_in_base_l {parent=calf_l, loc=<1.298 -5.2957 -0.0496>, rot=<0 0 0 -1>}
 deform_knee_in_l {parent=deform_knee_in_base_l, loc=<0 0 0>, rot=<1 0 0 0>, scale=<-0.2845 -1 -1>}
 deform_knee_out_base_l {parent=calf_l, loc=<0 4 0>, rot=<0 0 0 -1>}
 deform_knee_out_l {parent=deform_knee_out_base_l, loc=<0 0 0>, rot=<0 0 0 1>}
 deform_knee_inner_base_l {parent=calf_l, loc=<0 0 5>, rot=<0 0 0 -1>}
 deform_knee_inner_l {parent=deform_knee_inner_base_l, loc=<0 0 0>, rot=<0 0 0 1>}
 deform_knee_outer_base_l {parent=calf_l, loc=<0 0 -5>, rot=<0 0 0 -1>}
 deform_knee_outer_l {parent=deform_knee_outer_base_l, loc=<0 0 0>, rot=<0 0 0 1>}
 thigh_twist_01_l {parent=thigh_l, loc=<-21.6303 0 0>, rot=<0.0013 0 0 -1>}
 deform_groin_base_l {parent=thigh_l, loc=<-2.1908 8.1142 -1.7442>, rot=<0 0 0 -1>}
 deform_groin_l {parent=deform_groin_base_l, loc=<0 0 0>, rot=<0 0 0 1>}
 dyn_beltFront_anchor_a_master {parent=thigh_l, loc=<2.8775 10.1047 0.5447>, rot=<-0.3807 -0.465 -0.5142 -0.6119>}
 dyn_beltFront_anchor_a {parent=dyn_beltFront_anchor_a_master, loc=<0 0 0>, rot=<0 0 0 1>}
 thigh_r {parent=pelvis, loc=<-2.5693 -0.0109 7.7501>, rot=<0.0572 -0.0012 0.9982 0.0155>}
 calf_r {parent=thigh_r, loc=<39.0484 0 0>, rot=<0.0006 -0.0161 -0.0391 -0.9991>}
 foot_r {parent=calf_r, loc=<42.561 -0.1077 -0.4292>, rot=<0.0418 -0.0455 0.0343 -0.9975>}
 ball_r {parent=foot_r, loc=<6.1599 -13.8264 -0.6806>, rot=<0.0001 -0.0001 -0.7063 -0.7079>}
 deform_foot_in_base_r {parent=foot_r, loc=<0.3388 3.9009 -0.3197>, rot=<-0.0357 -0.7045 -0.0284 -0.7082>}
 deform_foot_in_r {parent=deform_foot_in_base_r, loc=<0 0 0>, rot=<0 0 0 1>}
 deform_foot_out_base_r {parent=foot_r, loc=<-1.951 -3.8422 -0.1665>, rot=<-0.0357 -0.7045 -0.0284 -0.7082>}
 deform_foot_out_r {parent=deform_foot_out_base_r, loc=<0 0 0>, rot=<0 0 0 1>}
 calf_twist_02_r {parent=calf_r, loc=<18.8064 0.0967 -0.0397>, rot=<0.0237 0.0095 -0.0035 -0.9997>}
 calf_twist_01_r {parent=calf_r, loc=<35.813 0.0967 -0.0397>, rot=<0.0237 0.0095 -0.0035 -0.9997>}
 deform_knee_in_base_r {parent=calf_r, loc=<-1.298 5.2957 0.0496>, rot=<0 0 0 -1>}
 deform_knee_in_r {parent=deform_knee_in_base_r, loc=<0 0 0>, rot=<0 0 0 1>}
 deform_knee_out_base_r {parent=calf_r, loc=<0 -4 0>, rot=<0 0 0 -1>}
 deform_knee_out_r {parent=deform_knee_out_base_r, loc=<0 0 0>, rot=<0 0 0 1>}
 deform_knee_outer_base_r {parent=calf_r, loc=<0 0 5>, rot=<0 0 0 -1>}
 deform_knee_outer_r {parent=deform_knee_outer_base_r, loc=<0 0 0>, rot=<0 0 0 1>}
 deform_knee_inner_base_r {parent=calf_r, loc=<0 0 -5>, rot=<0 0 0 -1>}
 deform_knee_inner_r {parent=deform_knee_inner_base_r, loc=<0 0 0>, rot=<0 0 0 1>}
 thigh_twist_01_r {parent=thigh_r, loc=<21.6303 0 0>, rot=<0.0013 0 0 -1>}
 deform_groin_base_r {parent=thigh_r, loc=<2.1908 -8.1142 1.7442>, rot=<0 0 0 -1>}
 deform_groin_r {parent=deform_groin_base_r, loc=<0 0 0>, rot=<0 0 0 1>}
 dyn_beltFront_root {parent=thigh_r, loc=<0 0 0>, rot=<0 0 0 1>}
 dyn_beltFront_anchor_d_master {parent=dyn_beltFront_root, loc=<8.66 -5.5574 9.119>, rot=<-0.1393 0.6542 0.2663 -0.694>}
 dyn_beltFront_anchor_d {parent=dyn_beltFront_anchor_d_master, loc=<0 0 0>, rot=<0 0 0 1>}
 dyn_beltFront_anchor_c_master {parent=dyn_beltFront_root, loc=<4.9826 -10.6259 1.6371>, rot=<-0.2623 0.4778 0.5461 -0.6362>}
 dyn_beltFront_anchor_c {parent=dyn_beltFront_anchor_c_master, loc=<0 0 0>, rot=<0 0 0 1>}
 dyn_shuriken_a_master {parent=dyn_beltFront_anchor_c, loc=<0.1511 2.4737 0.0144>, rot=<-0.0772 -0.0232 -0.0045 -0.9967>}
 dyn_shuriken_a_01 {parent=dyn_shuriken_a_master, loc=<0 0 0>, rot=<0 0 0 1>}
 dyn_shuriken_b_master {parent=dyn_beltFront_anchor_c, loc=<-0.1216 -2.4503 0.0795>, rot=<-0.046 -0.0477 0.1026 -0.9925>}
 dyn_shuriken_b_01 {parent=dyn_shuriken_b_master, loc=<0 0 0>, rot=<0 0 0 1>}
 deform_glute_base_l {parent=pelvis, loc=<-8.8064 -8.8222 -8.4287>, rot=<0.0275 0.0501 -0.4472 -0.8926>}
 deform_glute_l {parent=deform_glute_base_l, loc=<0 0 0>, rot=<0 0 0 1>}
 deform_glute_base_r {parent=pelvis, loc=<-8.8062 -8.8222 8.4287>, rot=<0.0501 -0.0275 0.8926 -0.4472>}
 deform_glute_r {parent=deform_glute_base_r, loc=<0 0 0>, rot=<0 0 0 1>}
 dyn_bagBelt_back_master {parent=pelvis, loc=<8.0212 -11.8172 -0.834>, rot=<-0.2307 -0.7168 -0.4842 0.4455>}
 dyn_bagBelt_back_01 {parent=dyn_bagBelt_back_master, loc=<0 0 0>, rot=<0 0 0 1>}
 dyn_bagBelt_back_02 {parent=dyn_bagBelt_back_01, loc=<4.7925 0 -0.1467>, rot=<-0.0043 0.1394 0.0716 -0.9876>}
 dyn_bagBelt_back_03 {parent=dyn_bagBelt_back_02, loc=<5.5886 0 -0.2697>, rot=<-0.0024 0.0571 0.0413 -0.9975>}
 dyn_bagBelt_back_04 {parent=dyn_bagBelt_back_03, loc=<5.1986 0 0>, rot=<0.9892 -0.0223 0.0454 -0.1375>}
 dyn_bagBelt_back_05 {parent=dyn_bagBelt_back_04, loc=<5.2093 0 0>, rot=<-0.006 -0.0543 0.1225 -0.991>}
 dyn_bagBelt_back_06 {parent=dyn_bagBelt_back_05, loc=<4.111 0 0>, rot=<0.0443 -0.1158 0.2014 -0.9716>}
 dyn_shuriken_c_master {parent=dyn_bagBelt_back_03, loc=<3.7047 0.0438 -0.3724>, rot=<0.4263 -0.4808 -0.5413 -0.5423>}
 dyn_shuriken_c {parent=dyn_shuriken_c_master, loc=<0 0 0>, rot=<0 0 0 1>}
 dyn_bagDrive_master {parent=pelvis, loc=<-2.5693 -0.0109 7.7501>, rot=<0.4442 -0.2119 0.8658 -0.0902>}
 dyn_bagDrive_01 {parent=dyn_bagDrive_master, loc=<0 0 0>, rot=<0 0 0 1>}
 dyn_bag_00 {parent=dyn_bagDrive_01, loc=<14.2422 0 0>, rot=<0 0.385 0 -0.9229>}
 dyn_bag_01 {parent=dyn_bag_00, loc=<3 0 0>, rot=<0 0 0 1>}
 dyn_rope_master {parent=pelvis, loc=<0 0 0>, rot=<-0.0127 -0.707 0.0127 -0.707>}
 dyn_rope_m_front_master {parent=dyn_rope_master, loc=<0 12.8189 0>, rot=<0 0 -0.7071 -0.7071>}
 dyn_rope_m_front_01 {parent=dyn_rope_m_front_master, loc=<0 0 0>, rot=<0 0 0 1>}
 dyn_rope_m_frontPart_r_master {parent=dyn_rope_m_front_01, loc=<1.3058 2.6729 -1.3502>, rot=<0.1136 -0.2129 -0.4065 -0.8812>}
 dyn_rope_m_frontPart_r_01 {parent=dyn_rope_m_frontPart_r_master, loc=<0 0 0>, rot=<0 0 0 1>}
 dyn_rope_m_frontPart_l_master {parent=dyn_rope_m_front_01, loc=<-0.9207 -5.5776 -1.4226>, rot=<0.0483 -0.2842 0.5863 -0.7571>}
 dyn_rope_m_frontPart_l_01 {parent=dyn_rope_m_frontPart_l_master, loc=<0 0 0>, rot=<0 0 0 1>}
 dyn_rope_rm_front_master {parent=dyn_rope_master, loc=<-7.6505 11.3218 1.0933>, rot=<0.0208 0.0869 0.8718 0.4817>}
 dyn_rope_rm_front_01 {parent=dyn_rope_rm_front_master, loc=<0 0 0>, rot=<0 0 0 1>}
 dyn_rope_mr_front_master {parent=dyn_rope_master, loc=<-13.8636 6.7124 2.9893>, rot=<0.0607 0.1535 0.9473 0.2745>}
 dyn_rope_mr_front_01 {parent=dyn_rope_mr_front_master, loc=<0 0 0>, rot=<0 0 0 1>}
 dyn_rope_r_front_master {parent=dyn_rope_master, loc=<-15.1936 -0.6488 5.1165>, rot=<0.0367 0.1759 0.9837 -0.0091>}
 dyn_rope_r_front_01 {parent=dyn_rope_r_front_master, loc=<0 0 0>, rot=<0 0 0 1>}
 dyn_lope_l_front_master {parent=dyn_rope_master, loc=<16.127 -0.8357 5.5125>, rot=<0.1647 0.0191 0.0136 -0.9861>}
 dyn_lope_l_front_01 {parent=dyn_lope_l_front_master, loc=<0 0 0>, rot=<0 0 0 -1>}
 dyn_lope_lm_front_master {parent=dyn_rope_master, loc=<8.0119 11.7619 0.6681>, rot=<0.0869 0.0208 -0.4817 -0.8718>}
 dyn_lope_lm_front_01 {parent=dyn_lope_lm_front_master, loc=<0 0 0>, rot=<0 0 0 1>}
 dyn_rope_ml_front_master {parent=dyn_rope_master, loc=<14.4022 6.9912 2.8461>, rot=<0.1535 0.0607 -0.2745 -0.9473>}
 dyn_rope_ml_front_01 {parent=dyn_rope_ml_front_master, loc=<0 0 0>, rot=<0 0 0 1>}
 dyn_rope_b_front_master {parent=dyn_rope_master, loc=<0.7879 -13.4631 9.4047>, rot=<0.1149 0.1149 0.6977 -0.6977>}
 dyn_rope_b_front_01 {parent=dyn_rope_b_front_master, loc=<0 0 0>, rot=<0 0 0 1>}
 dyn_rope_br_front_master {parent=dyn_rope_master, loc=<-10.6854 -9.3517 7.7269>, rot=<0.0646 0.1491 0.9053 -0.3924>}
 dyn_rope_br_front_01 {parent=dyn_rope_br_front_master, loc=<0 0 0>, rot=<0 0 0 1>}
 dyn_rope_bl_front_master {parent=dyn_rope_master, loc=<11.1989 -9.7962 8.1231>, rot=<0.1491 0.0646 0.3924 -0.9053>}
 dyn_rope_bl_front_01 {parent=dyn_rope_bl_front_master, loc=<0 0 0>, rot=<0 0 0 1>}
 dyn_beltFront_anchor_b_master {parent=pelvis, loc=<-2.93 11.0395 0.3798>, rot=<-0.4137 -0.3696 -0.5732 -0.603>}
 dyn_beltFront_anchor_b {parent=dyn_beltFront_anchor_b_master, loc=<0 0 0>, rot=<0 0 0 1>}
 attach {parent=root, loc=<0 0 0>, rot=<0 0 0 1>}
 ik_hand_root {parent=root, loc=<0 0 0>, rot=<0 0 0 1>}
 ik_hand_gun {parent=ik_hand_root, loc=<-51.9727 4.0883 102.6935>, rot=<0.6084 0.1793 0.4613 -0.6205>}
 ik_hand_r {parent=ik_hand_gun, loc=<0 0 0>, rot=<0 0 0 1>}
 ik_hand_l {parent=ik_hand_gun, loc=<53.0285 82.1781 35.2051>, rot=<-0.5102 -0.7906 -0.3387 0>}
 root_fp {parent=root, loc=<0 0 0>, rot=<0 0 0 1>}
 attach_fp {parent=root_fp, loc=<0 0 0>, rot=<0 0 0 1>}
 camera_root {parent=root_fp, loc=<0 0 0>, rot=<0 0 0 1>}
 camera {parent=camera_root, loc=<0 0 0>, rot=<0 0 0 1>}
 ik_foot_root {parent=root_fp, loc=<0 0 0>, rot=<0 0 0 1>}
 ik_foot_l {parent=ik_foot_root, loc=<15.2608 -4.4176 9.4432>, rot=<0.0357 0.7045 0.0284 -0.7082>}
 ik_foot_r {parent=ik_foot_root, loc=<-15.2608 -4.4176 9.4432>, rot=<0.7082 0.0284 -0.7045 0.0357>}
 attach_emote_01 {parent=root, loc=<0 0 0>, rot=<0 0 0 1>}
 attach_emote_02 {parent=root, loc=<0 0 0>, rot=<0 0 0 1>}
 attach_emote_03 {parent=root, loc=<0 0 0>, rot=<0 0 0 1>}
 warp_point_01 {parent=root, loc=<0 0 0>, rot=<0 0 0 1>}
 warp_point_02 {parent=root, loc=<0 0 0>, rot=<0 0 0 1>}
 warp_point_03 {parent=root, loc=<0 0 0>, rot=<0 0 0 1>}
 CrouchIdleAlpha {type=Curve, loc=<0 0 0>, rot=<0 0 0 1>}
]
variables [
 Abs Velocity {type=double, default=0}
 BagFollowLegFactor {type=double, default=0.3}
 Break_1_Elements_Yaw {type=double, default=0}
 Crouch Idle Alpha {type=double, default=0}
 DBNO Alpha {type=double, default=0}
 Is pot DBNO {type=bool, default=False}
 Is pot SkydiveFloating {type=bool, default=False}
]
