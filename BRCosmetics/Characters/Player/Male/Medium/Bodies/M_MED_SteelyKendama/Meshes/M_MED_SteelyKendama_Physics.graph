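# Physics asset for the SteelyKendama medium male player body cosmetic, giving physics to its dangling bag, belt and rope pieces.
# A six-link back-belt capsule chain joins the two-capsule bag, which also hangs from a drive bone and links to a front anchor with free angular motion.
# Two front ropes swing separately from their master bones.
# Unconstrained pelvis, thigh and right-glute shapes provide body collision volumes.
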PhysicsAsset /BRCosmetics/Characters/Player/Male/Medium/Bodies/M_MED_SteelyKendama/Meshes/M_MED_SteelyKendama_Physics
bodies {
 deform_glute_r [{shapeType=Sphere, radius=6.9556, shapeName=deform_glute_r_sphere, center=<-6.8987 1.0594 -2.055>}]
 dyn_bagBelt_back_01 [{
   shapeType=Capsule
   length=3.4954
   radius=2.5559
   shapeName=dyn_bagBelt_back_01_capsule
   center=<2.7769 0 -0.4528>
   rotation=<rot 0.0005 -89.9999 89.9997>
  }]
 dyn_bagBelt_back_02 [{
   shapeType=Capsule
   length=3.4954
   radius=2.5559
   shapeName=dyn_bagBelt_back_02_capsule
   center=<2.7691 0 -1.0037>
   rotation=<rot 0.0005 -89.9999 92.3207>
  }]
 dyn_bagBelt_back_03 [{
   shapeType=Capsule
   length=3.4954
   radius=2.5559
   shapeName=dyn_bagBelt_back_03_capsule
   center=<2.8288 0 -1.1126>
   rotation=<rot 0.0006 -89.9999 92.3357>
  }]
 dyn_bagBelt_back_04 [{
   shapeType=Capsule
   length=3.4954
   radius=2.5559
   shapeName=dyn_bagBelt_back_04_capsule
   center=<2.7583 0.4324 1.1086>
   rotation=<rot 0.0005 -90 96.7934>
  }]
 dyn_bagBelt_back_05 [{
   shapeType=Capsule
   length=3.4954
   radius=2.5559
   shapeName=dyn_bagBelt_back_05_capsule
   center=<2.7769 0 0.8172>
   rotation=<rot 0.0005 -89.9999 89.9997>
  }]
 dyn_bagBelt_back_06 [{
   shapeType=Capsule
   length=3.4954
   radius=2.5559
   shapeName=dyn_bagBelt_back_06_capsule
   center=<2.7769 0 1.308>
   rotation=<rot 0.0005 -89.9999 89.9997>
  }]
 dyn_bagBelt_back_master [{shapeType=Capsule, length=0, radius=2.3311, shapeName=dyn_bagBelt_back_master_capsule, rotation=<rot 0 -90 0>}]
 dyn_bagDrive_01 [{
   shapeType=Capsule
   length=5.79
   radius=2.0474
   shapeName=dyn_bagDrive_01_capsule
   center=<5.032 0 -0.0023>
   rotation=<rot 0.0001 -90 90.0261>
  }]
 dyn_bagDrive_master [{shapeType=Capsule, length=0.505, radius=0.505, shapeName=dyn_bagDrive_master_capsule, rotation=<rot 0 -90 0>}]
 dyn_bag_00 [
  {
   shapeType=Capsule
   length=14.1766
   radius=4.6378
   shapeName=dyn_bag_00_capsule
   center=<3.54 1.2333 3.1043>
   rotation=<rot 0 -7.6916 -84.4591>
  }
  {
   shapeType=Capsule
   length=16.1894
   radius=4.6378
   shapeName=dyn_bag_00_capsule1
   center=<3.983 0.6442 -6.3651>
   rotation=<rot -0.2597 -4.6631 -85.1033>
  }
 ]
 dyn_beltFront_anchor_d [{
   shapeType=Capsule
   length=3.5701
   radius=2.5735
   shapeName=dyn_beltFront_anchor_d_capsule
   center=<0.5465 0.9623 -0.0387>
   rotation=<rot -1.9513 0.851 -88.8265>
  }]
 dyn_rope_m_frontPart_l_01 [{
   shapeType=Capsule
   length=3.7528
   radius=1.5729
   shapeName=dyn_rope_m_frontPart_l_01_capsule
   center=<2.5243 0.0636 -0.3175>
   rotation=<rot 0 87.0402 -87.6332>
  }]
 dyn_rope_m_frontPart_l_master [{shapeType=Capsule, length=0.505, radius=0.505, shapeName=dyn_rope_m_frontPart_l_master_capsule, rotation=<rot 0 -90 0>}]
 dyn_rope_m_frontPart_r_01 [{
   shapeType=Capsule
   length=4.8534
   radius=2.3365
   shapeName=dyn_rope_m_frontPart_r_01_capsule
   center=<3.2986 0.4788 -0.5218>
   rotation=<rot 0 -80.3474 -70.4611>
  }]
 dyn_rope_m_frontPart_r_master [{shapeType=Capsule, length=0.505, radius=0.505, shapeName=dyn_rope_m_frontPart_r_master_capsule, rotation=<rot 0 -90 0>}]
 pelvis [{
   shapeType=Capsule
   length=15.9161
   radius=10.3945
   shapeName=pelvis_capsule
   center=<2.7852 -0.1225 -0.2421>
   rotation=<rot 3.2143 -96.0471 -32.461>
  }]
 thigh_l [{
   shapeType=Capsule
   length=5.5621
   radius=10.6646
   shapeName=thigh_l_capsule
   center=<-4.4418 -1.6507 2.2215>
   rotation=<rot 86.9627 179.9998 179.9999>
  }]
 thigh_r [{
   shapeType=Capsule
   length=29.5441
   radius=9.9302
   shapeName=thigh_r_capsule
   center=<14.6518 1.1702 -0.0363>
   rotation=<rot -84.375 -0.0005 0.0001>
  }]
}
constraints [
 {bone1Name=dyn_bagBelt_back_01, bone2Name=dyn_bagBelt_back_master, swing1LimitDegrees=10, swing2LimitDegrees=10, twistLimitDegrees=3}
 {bone1Name=dyn_bagBelt_back_02, bone2Name=dyn_bagBelt_back_01, swing1LimitDegrees=10, swing2LimitDegrees=10, twistLimitDegrees=3}
 {bone1Name=dyn_bagBelt_back_03, bone2Name=dyn_bagBelt_back_02, swing1LimitDegrees=10, swing2LimitDegrees=10, twistLimitDegrees=3}
 {bone1Name=dyn_bagBelt_back_04, bone2Name=dyn_bagBelt_back_03, swing1LimitDegrees=10, swing2LimitDegrees=10, twistLimitDegrees=3}
 {bone1Name=dyn_bagBelt_back_05, bone2Name=dyn_bagBelt_back_04, swing1LimitDegrees=10, swing2LimitDegrees=10, twistLimitDegrees=3}
 {bone1Name=dyn_bagBelt_back_06, bone2Name=dyn_bagBelt_back_05, swing1LimitDegrees=10, swing2LimitDegrees=10, twistLimitDegrees=3}
 {
  bone1Name=dyn_rope_m_frontPart_l_01
  bone2Name=dyn_rope_m_frontPart_l_master
  swing1LimitDegrees=15
  swing2LimitDegrees=15
  twistLimitDegrees=1
 }
 {
  bone1Name=dyn_rope_m_frontPart_r_01
  bone2Name=dyn_rope_m_frontPart_r_master
  swing1LimitDegrees=30
  swing2LimitDegrees=30
  twistLimitDegrees=1
 }
 {bone1Name=dyn_bagDrive_01, bone2Name=dyn_bagDrive_master, swing1LimitDegrees=30, swing2LimitDegrees=30, twistLimitDegrees=45}
 {bone1Name=dyn_bag_00, bone2Name=dyn_bagDrive_01, swing1LimitDegrees=10, swing2LimitDegrees=10, twistLimitDegrees=5}
 {bone1Name=dyn_bagBelt_back_06, bone2Name=dyn_bag_00, swing1LimitDegrees=15, swing2LimitDegrees=15, twistLimitDegrees=5}
 {
  bone1Name=dyn_bag_00
  bone2Name=dyn_beltFront_anchor_d
  swing1LimitDegrees=45
  swing1Motion=Free
  swing2LimitDegrees=45
  swing2Motion=Free
  twistLimitDegrees=45
  twistMotion=Free
 }
]
