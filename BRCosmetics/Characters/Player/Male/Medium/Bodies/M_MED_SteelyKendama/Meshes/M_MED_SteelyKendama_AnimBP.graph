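# Animation Blueprint for the SteelyKendama body cosmetic, adding secondary motion to its bag, belt, shurikens and knees.
# The copied parent pose runs through Control Rig, weighted bone constraints, knee AnimDynamics and RigidBody into a cached pose.
# bRigidBodyEnabled selects a crouch-driven post-sim rig with Look At, or constraint-only bag and shuriken attachment.
# The EventGraph updates crouch values, gravity override and WPO cloth-wind materials, and hides tails under long backblings.
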AnimBlueprint /BRCosmetics/Characters/Player/Male/Medium/Bodies/M_MED_SteelyKendama/Meshes/M_MED_SteelyKendama_AnimBP
target_skeleton=&/M_MED_SteelyKendama_Skeleton
graphs [
 AnimGraph {nodes [
   n1 = "Constraint - dyn_shuriken_b_master (PPPPPPPP)"!(
    ComponentPose
    Alpha=1
    ConstraintWeights_0=0.4008
    ConstraintWeights_1:float=0.0861
    ConstraintWeights_2:float=0.0049
    ConstraintWeights_3:float=0.0004
    ConstraintWeights_4:float=0.2217
    ConstraintWeights_5:float=0.2107
    ConstraintWeights_6:float=0.0009
    ConstraintWeights_7:float=0.0745
   ) -> (Pose) {
    alphaBoolBlend {blendOption=Linear}
    boneToModify {boneName=dyn_shuriken_b_master}
    constraintSetup [
     {targetBone {boneName=pelvis}}
     {targetBone {boneName=spine_01}}
     {targetBone {boneName=spine_02}}
     {targetBone {boneName=spine_03}}
     {targetBone {boneName=thigh_r}}
     {targetBone {boneName=thigh_twist_01_r}}
     {targetBone {boneName=deform_groin_l}}
     {targetBone {boneName=deform_groin_r}}
    ]
    constraintWeights [1,1,1,1,1,1,1,1]
   }
   "Constraint - dyn_shuriken_a_master (PPPPPPPP)"!(
    ComponentPose=@n1
    Alpha=1
    ConstraintWeights_0=0.3154
    ConstraintWeights_1:float=0.0622
    ConstraintWeights_2:float=0.0028
    ConstraintWeights_3:float=0.0001
    ConstraintWeights_4:float=0.2397
    ConstraintWeights_5:float=0.306
    ConstraintWeights_6:float=0.0732
    ConstraintWeights_7:float=0.0007
   ) -> (Pose) {
    alphaBoolBlend {blendOption=Linear}
    boneToModify {boneName=dyn_shuriken_a_master}
    constraintSetup [
     {targetBone {boneName=pelvis}}
     {targetBone {boneName=spine_01}}
     {targetBone {boneName=spine_02}}
     {targetBone {boneName=spine_03}}
     {targetBone {boneName=thigh_r}}
     {targetBone {boneName=thigh_twist_01_r}}
     {targetBone {boneName=deform_groin_r}}
     {targetBone {boneName=deform_glute_r}}
    ]
    constraintWeights [1,1,1,1,1,1,1,1]
   }
   n2 = "Transform (Modify) Bone - Bone: dyn_bag_00"(Alpha=1, Translation="-0.072,4.917,0.04", Rotation="1.588,22.196,-14.173", Scale="1,1,1") {
    rotationMode=BMM_Additive
    translationMode=BMM_Additive
    alphaBoolBlend {blendOption=Linear}
    boneToModify {boneName=dyn_bag_00}
    rotation=<rot 1.588 22.196 -14.173>
    translation=<-0.072 4.917 0.04>
   }
   n3 = "Transform (Modify) Bone - Bone: dyn_beltFront_anchor_d_master"(ComponentPose=@n2, Alpha=1, Translation="0.939,0.177,-0.377", Rotation="-11.539,-13.017,6.842", Scale="1,1,1") {
    rotationMode=BMM_Additive
    translationMode=BMM_Additive
    alphaBoolBlend {blendOption=Linear}
    boneToModify {boneName=dyn_beltFront_anchor_d_master}
    rotation=<rot -11.539 -13.017 6.842>
    translation=<0.939 0.177 -0.377>
   }
   n4 = "Transform (Modify) Bone - Bone: dyn_bag_00"(ComponentPose=@n3, Alpha=1, Translation="2.109,0.009,-2.000", Rotation="0.000,12.508,-2.947", Scale="1,1,1") {
    rotationMode=BMM_Additive
    translationMode=BMM_Additive
    alphaBoolBlend {blendOption=Linear}
    boneToModify {boneName=dyn_bag_00}
    rotation=<rot 0 12.508 -2.947>
    translation=<2.109 0.009 -2>
   }
   "Transform (Modify) Bone - Bone: dyn_bag_00"(ComponentPose=@n4, Alpha=1, Translation="-0.498,-1.069,1.689", Rotation="2.542,-8.216,9.519", Scale="1,1,1") {
    rotationMode=BMM_Additive
    translationMode=BMM_Additive
    alphaBoolBlend {blendOption=Linear}
    boneToModify {boneName=dyn_bag_00}
    rotation=<rot 2.542 -8.216 9.519>
    translation=<-0.498 -1.069 1.689>
   }
   "Transform (Modify) Bone - Bone: dyn_bag_00"(Alpha=1, Translation="3,0,0", Rotation="5,10,-10", Scale="1,1,1") {
    rotationMode=BMM_Additive
    translationMode=BMM_Additive
    alphaBoolBlend {blendOption=Linear}
    boneToModify {boneName=dyn_bag_00}
    rotation=<rot 5 10 -10>
    translation {x=3}
   }
   n5 = Control Rig!(
    Source=Use cached pose 'InitSetup'()
    Alpha
    Crouch Idle:double=Get Crouch Idle() -> (Crouch Idle:float)
    Crouch Walk:double=Get Crouch Move() -> (Crouch Move:double)
    Crouch Walk Side:double=Get X Movement() -> (X Movement:double)
   ) -> (Pose) {
    alphaBoolBlend {blendOption=Linear}
    controlRigAssetReference {blueprintRigClass=&/M_MED_SteelyKendama_PostSim_CtrlRig.M_MED_SteelyKendama_PostSim_CtrlRig_C}
   }
   n6 = Look At - Bone: dyn_beltFront_anchor_d(ComponentPose=Local To Component(LocalPose=@n5), Alpha=1) {
    bUseLookUpAxis=true
    interpolationType=Cubic
    alphaBoolBlend {blendOption=Linear}
    boneToModify {boneName=dyn_beltFront_anchor_d}
    lookAtLocation {x=50}
    lookAtTarget {boneReference {boneName=dyn_bagDrive_01}}
    lookUp_Axis {axis {x=0, z=1}}
   }
   n7 = "Constraint - dyn_shuriken_b_master (PPPPPPPP)"(
    ComponentPose=@n6
    Alpha=1
    ConstraintWeights_0=0.0001
    ConstraintWeights_1=0
    ConstraintWeights_2=0
    ConstraintWeights_3=0
    ConstraintWeights_4=0
    ConstraintWeights_5=0.7241
    ConstraintWeights_6=0.0283
    ConstraintWeights_7=0.2475
   ) {
    alphaBoolBlend {blendOption=Linear}
    boneToModify {boneName=dyn_shuriken_b_master}
    constraintSetup [
     {targetBone {boneName=pelvis}}
     {targetBone {boneName=spine_01}}
     {targetBone {boneName=thigh_l}}
     {targetBone {boneName=thigh_twist_01_l}}
     {targetBone {boneName=deform_groin_l}}
     {targetBone {boneName=dyn_beltFront_anchor_c}}
     {targetBone {boneName=dyn_beltFront_anchor_d}}
     {targetBone {boneName=dyn_beltFront_anchor_b}}
    ]
    constraintWeights [1,1,1,1,1,1,1,1]
   }
   n8 = "Constraint - dyn_shuriken_a_master (PPPPPPPP)"(
    ComponentPose=@n7
    Alpha=1
    ConstraintWeights_0=0
    ConstraintWeights_1=0
    ConstraintWeights_2=0
    ConstraintWeights_3=0
    ConstraintWeights_4=0
    ConstraintWeights_5=0.7353
    ConstraintWeights_6=0.2468
    ConstraintWeights_7=0.0179
   ) {
    alphaBoolBlend {blendOption=Linear}
    boneToModify {boneName=dyn_shuriken_a_master}
    constraintSetup [
     {targetBone {boneName=pelvis}}
     {targetBone {boneName=spine_01}}
     {targetBone {boneName=thigh_l}}
     {targetBone {boneName=thigh_twist_01_l}}
     {targetBone {boneName=dyn_bag_00}}
     {targetBone {boneName=dyn_beltFront_anchor_c}}
     {targetBone {boneName=dyn_beltFront_anchor_d}}
     {targetBone {boneName=dyn_beltFront_anchor_b}}
    ]
    constraintWeights [1,1,1,1,1,1,1,1]
   }
   n9 = "Constraint - dyn_shuriken_c_master (PPPPPPPP)"!(
    ComponentPose=@n8
    Alpha=1
    ConstraintWeights_0=0.0016
    ConstraintWeights_1:float=0.0005
    ConstraintWeights_2:float=0.0058
    ConstraintWeights_3:float=0.0909
    ConstraintWeights_4:float=0.3784
    ConstraintWeights_5:float=0.4255
    ConstraintWeights_6:float=0.0866
    ConstraintWeights_7:float=0.0108
   ) -> (Pose) {
    alphaBoolBlend {blendOption=Linear}
    boneToModify {boneName=dyn_shuriken_c_master}
    constraintSetup [
     {targetBone {boneName=pelvis}}
     {targetBone {boneName=spine_01}}
     {targetBone {boneName=dyn_bagBelt_back_01}}
     {targetBone {boneName=dyn_bagBelt_back_02}}
     {targetBone {boneName=dyn_bagBelt_back_03}}
     {targetBone {boneName=dyn_bagBelt_back_04}}
     {targetBone {boneName=dyn_bagBelt_back_05}}
     {targetBone {boneName=dyn_bagBelt_back_06}}
    ]
    constraintWeights [1,1,1,1,1,1,1,1]
   }
   n10 = "Constraint - dyn_bag_00 (PPPPPPPP)"!(
    ComponentPose=Local To Component(LocalPose=Use cached pose 'InitSetup'())
    Alpha=1
    ConstraintWeights_0=0.1914
    ConstraintWeights_1:float=0.0134
    ConstraintWeights_2:float=0.2381
    ConstraintWeights_3:float=0.38
    ConstraintWeights_4:float=0.0036
    ConstraintWeights_5:float=0.0015
    ConstraintWeights_6:float=0.0077
    ConstraintWeights_7:float=0.1644
   ) -> (Pose) {
    alphaBoolBlend {blendOption=Linear}
    boneToModify {boneName=dyn_bag_00}
    constraintSetup [
     {targetBone {boneName=pelvis}}
     {targetBone {boneName=spine_01}}
     {targetBone {boneName=thigh_r}}
     {targetBone {boneName=thigh_twist_01_r}}
     {targetBone {boneName=calf_r}}
     {targetBone {boneName=deform_knee_in_r}}
     {targetBone {boneName=deform_groin_r}}
     {targetBone {boneName=deform_glute_r}}
    ]
    constraintWeights [1,1,1,1,1,1,1,1]
   }
   n11 = "Constraint - dyn_shuriken_c_master (PPPPPPPP)"(
    ComponentPose=@n10
    Alpha=1
    ConstraintWeights_0=0.5671
    ConstraintWeights_1=0.0602
    ConstraintWeights_2=0.0004
    ConstraintWeights_3=0.0088
    ConstraintWeights_4=0.0384
    ConstraintWeights_5=0.0453
    ConstraintWeights_6=0.0043
    ConstraintWeights_7=0.2755
   ) {
    alphaBoolBlend {blendOption=Linear}
    boneToModify {boneName=dyn_shuriken_c_master}
    constraintSetup [
     {targetBone {boneName=pelvis}}
     {targetBone {boneName=spine_01}}
     {targetBone {boneName=spine_03}}
     {targetBone {boneName=thigh_twist_01_l}}
     {targetBone {boneName=thigh_r}}
     {targetBone {boneName=thigh_twist_01_r}}
     {targetBone {boneName=deform_glute_l}}
     {targetBone {boneName=deform_glute_r}}
    ]
    constraintWeights [1,1,1,1,1,1,1,1]
   }
   n12 = "Constraint - dyn_bag_00 (P)"!(ComponentPose=@n11, bAlphaBoolEnabled:bool=True, ConstraintWeights_0=1) -> (Pose) {
    alphaInputType=Bool
    alphaBoolBlend {blendOption=Linear}
    boneToModify {boneName=dyn_bag_00}
    constraintSetup [{targetBone {boneName=thigh_r}}]
    constraintWeights [1]
   }
   Output Pose(Result=Blend Poses by bool(BlendPose_0=Component To Local(ComponentPose=@n9), BlendPose_1=Component To Local(ComponentPose=@n12), bActiveValue=Get bRigidBodyEnabled()) {blendType=Linear}) {layerGroup=None}
   n13 = Control Rig!(Source=Copy Pose From Mesh() {bUseAttachedParent=true}, Alpha, Abs Velocity:double=Get Abs Velocity() -> (Abs Velocity:double)) -> (Pose) {
    alphaBoolBlend {blendOption=Linear}
    controlRigAssetReference {blueprintRigClass=&/M_MED_SteelyKendama_CtrlRig.M_MED_SteelyKendama_CtrlRig_C}
   }
   n14 = "Constraint - Tail_root (PPPPPPPP)"!(
    ComponentPose=Local To Component(LocalPose=@n13)
    Alpha=1
    ConstraintWeights_0=0.8854
    ConstraintWeights_1:float=0.0918
    ConstraintWeights_2:float=0.006
    ConstraintWeights_3:float=0.0074
    ConstraintWeights_4:float=0
    ConstraintWeights_5:float=0.0085
    ConstraintWeights_6:float=0.0005
    ConstraintWeights_7:float=0.0005
   ) -> (Pose) {
    alphaBoolBlend {blendOption=Linear}
    boneToModify {boneName=Tail_root}
    constraintSetup [
     {targetBone {boneName=pelvis}}
     {targetBone {boneName=spine_01}}
     {targetBone {boneName=spine_03}}
     {targetBone {boneName=thigh_twist_01_l}}
     {targetBone {boneName=thigh_r}}
     {targetBone {boneName=thigh_twist_01_r}}
     {targetBone {boneName=deform_glute_l}}
     {targetBone {boneName=deform_glute_r}}
    ]
    constraintWeights [1,1,1,1,1,1,1,1]
   }
   n15 = Anim Dynamics - Bone: deform_knee_inner_base_l(ComponentPose=@n14, Alpha=1, ExternalForce="0,0,0") {
    angularDampingOverride=0.9999
    angularSpringConstant=1000
    bAngularSpring=true
    bLinearSpring=true
    bOverrideAngularDamping=true
    bOverrideLinearDamping=true
    bUseGravityOverride=true
    linearDampingOverride=0.999
    linearSpringConstant=100
    simulationSpace=RootRelative
    alphaBoolBlend {blendOption=Linear}
    boundBone {boneName=deform_knee_inner_base_l}
    componentLinearAccScale=<1 1 1>
    componentLinearVelScale=<1 1 1>
    gravityOverride {z=-980}
    physicsBodyDefinitions [{
      boundBone {boneName=deform_knee_inner_base_l}
      boxExtents=<10 10 10>
      constraintSetup {angularLimitsMax=<3 3 3>, angularLimitsMin=<-3 -3 -3>, angularTarget=<0 0 0>, linearAxesMax=<0 5 5>, linearAxesMin=<-5 -5 -5>}
      localJointOffset=<0.1 0 0>
     }]
   }
   n16 = Anim Dynamics - Bone: deform_knee_outer_base_l(ComponentPose=@n15, Alpha=1, ExternalForce="0,0,0") {
    angularDampingOverride=0.9999
    angularSpringConstant=1000
    bAngularSpring=true
    bLinearSpring=true
    bOverrideAngularDamping=true
    bOverrideLinearDamping=true
    bUseGravityOverride=true
    linearDampingOverride=0.999
    linearSpringConstant=100
    simulationSpace=RootRelative
    alphaBoolBlend {blendOption=Linear}
    boundBone {boneName=deform_knee_outer_base_l}
    componentLinearAccScale=<1 1 1>
    componentLinearVelScale=<1 1 1>
    gravityOverride {z=-980}
    physicsBodyDefinitions [{
      boundBone {boneName=deform_knee_outer_base_l}
      boxExtents=<10 10 10>
      constraintSetup {angularLimitsMax=<3 3 3>, angularLimitsMin=<-3 -3 -3>, angularTarget=<0 0 0>, linearAxesMax=<0 5 5>, linearAxesMin=<-5 -5 -2>}
      localJointOffset=<0.1 0 0>
     }]
   }
   n17 = Anim Dynamics - Bone: deform_knee_in_base_l(ComponentPose=@n16, Alpha=1, ExternalForce="0,0,0") {
    angularDampingOverride=0.9999
    angularSpringConstant=1000
    bAngularSpring=true
    bLinearSpring=true
    bOverrideAngularDamping=true
    bOverrideLinearDamping=true
    bUseGravityOverride=true
    linearDampingOverride=0.999
    linearSpringConstant=100
    simulationSpace=RootRelative
    alphaBoolBlend {blendOption=Linear}
    boundBone {boneName=deform_knee_in_base_l}
    componentLinearAccScale=<1 1 1>
    componentLinearVelScale=<1 1 1>
    gravityOverride {z=-980}
    physicsBodyDefinitions [{
      boundBone {boneName=deform_knee_in_base_l}
      boxExtents=<10 10 10>
      constraintSetup {angularLimitsMax=<3 3 3>, angularLimitsMin=<-3 -3 -3>, angularTarget=<0 0 0>, linearAxesMax=<0 5 5>, linearAxesMin=<-5 -5 -5>}
      localJointOffset=<0.1 0 0>
     }]
   }
   n18 = Anim Dynamics - Bone: deform_knee_out_base_l(ComponentPose=@n17, Alpha=1, ExternalForce="0,0,0") {
    angularDampingOverride=0.9999
    angularSpringConstant=1000
    bAngularSpring=true
    bLinearSpring=true
    bOverrideAngularDamping=true
    bOverrideLinearDamping=true
    bUseGravityOverride=true
    linearDampingOverride=0.999
    linearSpringConstant=100
    simulationSpace=RootRelative
    alphaBoolBlend {blendOption=Linear}
    boundBone {boneName=deform_knee_out_base_l}
    componentLinearAccScale=<1 1 1>
    componentLinearVelScale=<1 1 1>
    gravityOverride {z=-980}
    physicsBodyDefinitions [{
      boundBone {boneName=deform_knee_out_base_l}
      boxExtents=<10 10 10>
      constraintSetup {angularLimitsMax=<3 3 3>, angularLimitsMin=<-3 -3 -3>, angularTarget=<0 0 0>, linearAxesMax=<0 5 5>, linearAxesMin=<-5 -5 -5>}
      localJointOffset=<0.1 0 0>
     }]
   }
   n19 = Anim Dynamics - Bone: deform_knee_inner_base_r(ComponentPose=@n18, Alpha=1, ExternalForce="0,0,0") {
    angularDampingOverride=0.9999
    angularSpringConstant=1000
    bAngularSpring=true
    bLinearSpring=true
    bOverrideAngularDamping=true
    bOverrideLinearDamping=true
    bUseGravityOverride=true
    linearDampingOverride=0.999
    linearSpringConstant=100
    simulationSpace=RootRelative
    alphaBoolBlend {blendOption=Linear}
    boundBone {boneName=deform_knee_inner_base_r}
    componentLinearAccScale=<1 1 1>
    componentLinearVelScale=<1 1 1>
    gravityOverride {z=-980}
    physicsBodyDefinitions [{
      boundBone {boneName=deform_knee_inner_base_r}
      boxExtents=<10 10 10>
      constraintSetup {angularLimitsMax=<3 3 3>, angularLimitsMin=<-3 -3 -3>, angularTarget=<0 0 0>, linearAxesMax=<5 5 5>, linearAxesMin=<0 -5 -5>}
      localJointOffset=<0.1 0 0>
     }]
   }
   n20 = Anim Dynamics - Bone: deform_knee_outer_base_r(ComponentPose=@n19, Alpha=1, ExternalForce="0,0,0") {
    angularDampingOverride=0.9999
    angularSpringConstant=1000
    bAngularSpring=true
    bLinearSpring=true
    bOverrideAngularDamping=true
    bOverrideLinearDamping=true
    bUseGravityOverride=true
    linearDampingOverride=0.999
    linearSpringConstant=100
    simulationSpace=RootRelative
    alphaBoolBlend {blendOption=Linear}
    boundBone {boneName=deform_knee_outer_base_r}
    componentLinearAccScale=<1 1 1>
    componentLinearVelScale=<1 1 1>
    gravityOverride {z=-980}
    physicsBodyDefinitions [{
      boundBone {boneName=deform_knee_outer_base_r}
      boxExtents=<10 10 10>
      constraintSetup {angularLimitsMax=<3 3 3>, angularLimitsMin=<-3 -3 -3>, angularTarget=<0 0 0>, linearAxesMax=<5 5 5>, linearAxesMin=<0 -5 -2>}
      localJointOffset=<0.1 0 0>
     }]
   }
   n21 = Anim Dynamics - Bone: deform_knee_in_base_r(ComponentPose=@n20, Alpha=1, ExternalForce="0,0,0") {
    angularDampingOverride=0.9999
    angularSpringConstant=1000
    bAngularSpring=true
    bLinearSpring=true
    bOverrideAngularDamping=true
    bOverrideLinearDamping=true
    bUseGravityOverride=true
    linearDampingOverride=0.999
    linearSpringConstant=100
    simulationSpace=RootRelative
    alphaBoolBlend {blendOption=Linear}
    boundBone {boneName=deform_knee_in_base_r}
    componentLinearAccScale=<1 1 1>
    componentLinearVelScale=<1 1 1>
    gravityOverride {z=-980}
    physicsBodyDefinitions [{
      boundBone {boneName=deform_knee_in_base_r}
      boxExtents=<10 10 10>
      constraintSetup {angularLimitsMax=<3 3 3>, angularLimitsMin=<-3 -3 -3>, angularTarget=<0 0 0>, linearAxesMax=<5 5 5>, linearAxesMin=<0 -5 -5>}
      localJointOffset=<0.1 0 0>
     }]
   }
   n22 = Anim Dynamics - Bone: deform_knee_out_base_r(ComponentPose=@n21, Alpha=1, ExternalForce="0,0,0") {
    angularDampingOverride=0.9999
    angularSpringConstant=1000
    bAngularSpring=true
    bLinearSpring=true
    bOverrideAngularDamping=true
    bOverrideLinearDamping=true
    bUseGravityOverride=true
    linearDampingOverride=0.999
    linearSpringConstant=100
    simulationSpace=RootRelative
    alphaBoolBlend {blendOption=Linear}
    boundBone {boneName=deform_knee_out_base_r}
    componentLinearAccScale=<1 1 1>
    componentLinearVelScale=<1 1 1>
    gravityOverride {z=-980}
    physicsBodyDefinitions [{
      boundBone {boneName=deform_knee_out_base_r}
      boxExtents=<10 10 10>
      constraintSetup {angularLimitsMax=<3 3 3>, angularLimitsMin=<-3 -3 -3>, angularTarget=<0 0 0>, linearAxesMax=<5 5 5>, linearAxesMin=<0 -5 -5>}
      localJointOffset=<0.1 0 0>
     }]
   }
   n23 = "Constraint - dyn_beltFront_anchor_a_master (PPPPPPP)"!(
    ComponentPose=@n22
    Alpha=1
    ConstraintWeights_0=0.5577
    ConstraintWeights_1:float=0.1567
    ConstraintWeights_2:float=0.0171
    ConstraintWeights_3:float=0.0057
    ConstraintWeights_4:float=0.0974
    ConstraintWeights_5:float=0.1136
    ConstraintWeights_6:float=0.0517
   ) -> (Pose) {
    alphaBoolBlend {blendOption=Linear}
    boneToModify {boneName=dyn_beltFront_anchor_a_master}
    constraintSetup [
     {targetBone {boneName=pelvis}}
     {targetBone {boneName=spine_01}}
     {targetBone {boneName=spine_02}}
     {targetBone {boneName=spine_03}}
     {targetBone {boneName=thigh_l}}
     {targetBone {boneName=thigh_twist_01_l}}
     {targetBone {boneName=deform_groin_l}}
    ]
    constraintWeights [1,1,1,1,1,1,1]
   }
   n24 = "Constraint - dyn_beltFront_anchor_b_master (PPPPPPPP)"!(
    ComponentPose=@n23
    Alpha=1
    ConstraintWeights_0=0.5811
    ConstraintWeights_1:float=0.1259
    ConstraintWeights_2:float=0.0651
    ConstraintWeights_3:float=0.0416
    ConstraintWeights_4:float=0.0219
    ConstraintWeights_5:float=0.1091
    ConstraintWeights_6:float=0.0275
    ConstraintWeights_7:float=0.0278
   ) -> (Pose) {
    alphaBoolBlend {blendOption=Linear}
    boneToModify {boneName=dyn_beltFront_anchor_b_master}
    constraintSetup [
     {targetBone {boneName=pelvis}}
     {targetBone {boneName=spine_01}}
     {targetBone {boneName=thigh_l}}
     {targetBone {boneName=thigh_twist_01_l}}
     {targetBone {boneName=thigh_r}}
     {targetBone {boneName=thigh_twist_01_r}}
     {targetBone {boneName=deform_groin_l}}
     {targetBone {boneName=deform_groin_r}}
    ]
    constraintWeights [1,1,1,1,1,1,1,1]
   }
   n25 = "Constraint - dyn_beltFront_anchor_c_master (PPPPPPPP)"!(
    ComponentPose=@n24
    Alpha=1
    ConstraintWeights_0=0.3481
    ConstraintWeights_1:float=0.0728
    ConstraintWeights_2:float=0.0038
    ConstraintWeights_3:float=0.0003
    ConstraintWeights_4:float=0.234
    ConstraintWeights_5:float=0.2643
    ConstraintWeights_6:float=0.0766
    ConstraintWeights_7:float=0.0001
   ) -> (Pose) {
    alphaBoolBlend {blendOption=Linear}
    boneToModify {boneName=dyn_beltFront_anchor_c_master}
    constraintSetup [
     {targetBone {boneName=pelvis}}
     {targetBone {boneName=spine_01}}
     {targetBone {boneName=spine_02}}
     {targetBone {boneName=spine_03}}
     {targetBone {boneName=thigh_r}}
     {targetBone {boneName=thigh_twist_01_r}}
     {targetBone {boneName=deform_groin_r}}
     {targetBone {boneName=deform_glute_r}}
    ]
    constraintWeights [1,1,1,1,1,1,1,1]
   }
   n26 = "Constraint - dyn_beltFront_anchor_d_master (PPPPPPPP)"!(
    ComponentPose=@n25
    Alpha=1
    ConstraintWeights_0=0.2223
    ConstraintWeights_1:float=0.0295
    ConstraintWeights_2:float=0.0003
    ConstraintWeights_3:float=0.0126
    ConstraintWeights_4:float=0.2614
    ConstraintWeights_5:float=0.415
    ConstraintWeights_6:float=0.0438
    ConstraintWeights_7:float=0.0149
   ) -> (Pose) {
    alphaBoolBlend {blendOption=Linear}
    boneToModify {boneName=dyn_beltFront_anchor_d_master}
    constraintSetup [
     {targetBone {boneName=pelvis}}
     {targetBone {boneName=spine_01}}
     {targetBone {boneName=spine_02}}
     {targetBone {boneName=thigh_twist_01_l}}
     {targetBone {boneName=thigh_r}}
     {targetBone {boneName=thigh_twist_01_r}}
     {targetBone {boneName=deform_groin_r}}
     {targetBone {boneName=deform_glute_r}}
    ]
    constraintWeights [1,1,1,1,1,1,1,1]
   }
   n27 = RigidBody!(ComponentPose=@n26, Alpha, OverrideWorldGravity:Vector="0,0,0", ExternalForce) -> (Pose) {
    bClampLinearTranslationLimitToRefPose=true
    bOverrideWorldGravity=true
    overridePhysicsAsset=&/M_MED_SteelyKendama_Physics
    simulationSpace=BaseBoneSpace
    alphaBoolBlend {blendOption=Linear}
    baseBoneRef {boneName=pelvis}
    componentLinearAccScale=<0.4 0.4 0.4>
    componentLinearVelScale=<0.4 0.4 0.4>
    simSpaceSettings {maxAngularAcceleration=500, maxAngularVelocity=3, maxLinearAcceleration=400, maxLinearVelocity=250, externalLinearDragV=<0.2 0.2 0.2>}
   }
   Save cached pose 'InitSetup'(Pose=Component To Local(ComponentPose=@n27)) {cacheName=InitSetup}
  ]}
 EventGraph {nodes [
   n1 = Event BlueprintUpdateAnimation()
   > n2 = Sequence!(execute) -> (then_0, then_1, then_2:Exec, then_3:Exec, then_4:Exec)
   > n3 = Gravity Override State Function DEPRECATED(
    CharacterPartAnimInstance=Self-Reference()
    GameTimeInSeconds=GetGameTimeInSeconds()
    GravityOverrideParams=Get Gravity Override Params()
    RBAN Params=Get RBAN Params()
    DebugDraw=Get doDebug() -> (doDebug:bool)
   )
   > Set GravityOverride(GravityOverride=@n3)
   > n4 = Set Abs Velocity(Abs Velocity:double=Vector Length(A=Get PawnVelocity())) -> (Output_Get:double)
   > Branch(Condition=Get doDebug())
   > PrintString!(execute, InString=Append(A=@n3.CurrentStateOut, B="To String (Float)"(InDouble=@n4)), bPrintToScreen, bPrintToLog, TextColor, Duration, Key) -> (then)
   n5 = Branch(
    execute=@n2.then_1
    Condition=OR Boolean!(A=Get bIsSkydiveFloating(), B=Get bIsSkydiveDiveMode(), C:bool=Get bIsParachuteOpened()) -> (ReturnValue)
   )
   > Set Velocity Interp Speed(Velocity Interp Speed:double=1) -> (Output_Get:double)
   > Set WPO Wind(WPO Wind:double=1) -> (Output_Get:double)
   > Set Turbulence Noise(Turbulence Noise:float=75) -> (Output_Get:float)
   > Set CrouchIdleAlphaInv(CrouchIdleAlphaInv:float=0) -> (Output_Get:float)
   Set Velocity Interp Speed(execute=@n5.else, Velocity Interp Speed=0)
   > Set WPO Wind(WPO Wind=0)
   > Set Turbulence Noise(Turbulence Noise=0)
   > Set CrouchIdleAlphaInv(CrouchIdleAlphaInv=MapRangeClamped(Value=GetCurveValue(CurveName=CrouchIdleAlpha), InRangeA=0, InRangeB=1, OutRangeA=1, OutRangeB=0))
   n6 = Get bIsSliding()
   n7 = Branch(execute=@n2.then_4, Condition=AND Boolean(A=Get bIsCrouching(), B=NOT Boolean(A=@n6)))
   > Set Crouch Idle(Crouch Idle:float=GetCurveValue(CurveName=CrouchIdleAlpha)) -> (Output_Get:float)
   n8 = Get Abs Velocity()
   > n9 = Set Crouch Move(Crouch Move:double=float * float!(
     A=Get Crouch Alpha() -> (Crouch Alpha:double)
     B=Get CrouchIdleAlphaInv() -> (CrouchIdleAlphaInv:float)
     C:double=MapRangeClamped(Value=@n8, InRangeA=0, InRangeB=178)
     D:double=MapRangeClamped(Value=@n8, InRangeA=700, InRangeB=250)
    ) -> (ReturnValue) {is=title K2Node_PromotableOperator}) -> (Output_Get:double)
   > n10 = PlayerMovement(AnimInstance=Self-Reference(), XDotProductCoefficient=1, YDotProductCoefficient=1, ZDotProductCoefficient=1)
   > Set X Movement(X Movement:double=float * float!(A="Absolute (Float)"(A=@n10), B=@n9) -> (ReturnValue)) -> (Output_Get:double)
   Set Crouch Move(execute=@n7.else, Crouch Move=0)
   > Set Crouch Idle(Crouch Idle=0)
   > Set X Movement(X Movement=0)
   n11 = Get Mesh!() -> (Mesh)
   n12 = Get Overlay Mat Instance Dyanmic() -> (Overlay Mat Instance Dyanmic:Material Instance Dynamic Object Reference)
   n13 = Get Material Instance Dynamic() -> (Material Instance Dynamic:Material Instance Dynamic Object Reference)
   Set WPO Globals Function(
    execute=@n2.then_2
    Material=@n13
    Overlay Material=@n12
    Skeletal Mesh=@n11
    "Local Noise Transform  Location"="Get Local Noise Transform  Location"() -> ("Local Noise Transform  Location":Vector)
    Local Noise Offset=Get Local Noise Offset()
   )
   > Set WPO Regional Function!(
    execute
    Material:Material Instance Dynamic Object Reference=@n13
    Overlay Material:Material Instance Dynamic Object Reference=@n12
    Skeletal Mesh:Skeletal Mesh Component Object Reference=@n11
    Velocity Transform:Name=calf_r
    Velocity Interp Speed:double=Get Velocity Interp Speed()
    Collision Transform:Name=calf_r
    Collision Transform Axis:EAxis Enum
    Invert Collision Transform Axis:bool=true
    Delta Time X:double=@n1.DeltaTimeX
    Previous Velocity Transform Location:"Vector (by ref)"=Get R2 Velocity Transform Location()
    Previous Velocity Transform Velocity:"Vector (by ref)"=Get R2 Velocity Transform Velocity()
    Region:int=1
   ) -> (then)
   > Set WPO Regional Function(
    Material=@n13
    Overlay Material=@n12
    Skeletal Mesh=@n11
    Velocity Transform=calf_l
    Velocity Interp Speed=Get Velocity Interp Speed()
    Collision Transform=calf_l
    Invert Collision Transform Axis=false
    Delta Time X=@n1.DeltaTimeX
    Previous Velocity Transform Location=Get L Previous Velocity Transform Location() -> (L Previous Velocity Transform Location:Vector)
    Previous Velocity Transform Velocity=Get L Previous Velocity Transform Velocity() -> (L Previous Velocity Transform Velocity:Vector)
    Region=2
   )
   > SetScalarParameterValue(self=[@n13, @n12], ParameterName=ClothSim_WindStrength, Value=Get WPO Wind() -> (WPO Wind:double))
   > SetScalarParameterValue(self=[@n13, @n12], ParameterName=ClothSim_TurbulenceNoiseTileDivisor, Value=Get Turbulence Noise() -> (Turbulence Noise:float))
   Set LOD2andLower(execute=@n2.then_3, LOD2andLower:bool="integer >= integer"(A=Get CurrentLOD(), B=2)) -> (Output_Get:bool)
   Event OnEndCrouching()
   > Set Crouch Alpha(Crouch Alpha:double=0) -> (Output_Get:double)
   Event OnBeginCrouching()
   > Set Crouch Alpha(Crouch Alpha=1)
   Event OnCharacterCustomizationLoaded()
   > n14 = Sequence!(execute) -> (then_0, then_1)
   n15 = Cast To PlayerPawn_Athena(execute=@n14.then_1, Object=TryGetPawnOwner())
   > Branch(Condition=CheckForTail!(self:Custom Character Part Anim Instance Object Reference) -> (bHasTail:bool))
   > n16 = Branch(Condition=CheckForBackbling().bIsBackblingLong)
   n17 = GetSkeletalMeshForPartType(self=@n15, PartType=MiscOrTail)
   > SetVisibility!(execute, self:Scene Component Object Reference=@n17, bNewVisibility:bool=false, bPropagateToChildren:bool=false) -> (then)
   SetVisibility(execute=@n16.else, self=@n17, bNewVisibility=true, bPropagateToChildren=false)
   n18 = GetPartSkeletalMeshComponent()
   n19 = Set Mesh(execute=@n14.then_0, Mesh:Skeletal Mesh Component Object Reference=@n18) -> (Output_Get:Skeletal Mesh Component Object Reference)
   > n20 = Get Dynamic Material Instance(SkeletalMeshComponent=@n19)
   > Set Material Instance Dynamic(Material Instance Dynamic:Material Instance Dynamic Object Reference=@n20) -> (Output_Get:Material Instance Dynamic Object Reference)
   > n21 = GetOrCreateDynamicOverlayMaterial(MeshComponent=@n18, MaterialName=None)
   > Set Overlay Mat Instance Dyanmic(Overlay Mat Instance Dyanmic:Material Instance Dynamic Object Reference=@n21) -> (Output_Get:Material Instance Dynamic Object Reference)
  ]}
]
variables [
 Abs Velocity {type=number}
 Alpha {type=number, default=1}
 Bag Follow Leg Factor {type=number, default=0.3}
 Character Part Anim Instance {type=CustomCharacterPartAnimInstance, default=null}
 Crouch Alpha {type=number}
 Crouch Idle {type=number}
 Crouch Move {type=number}
 CrouchIdleAlphaInv {type=number}
 doDebug {type=boolean}
 Gravity Override Params {type=GravityOverrideParamsStruct}
 GravityOverride {type=Vector}
 L Previous Velocity Transform Location {type=Vector}
 L Previous Velocity Transform Velocity {type=Vector}
 Local Noise Offset {type=Vector}
 "Local Noise Transform  Location" {type=Vector}
 LOD2andLower {type=boolean}
 Material Instance Dynamic {type=MaterialInstanceDynamic}
 Mesh {type=SkeletalMeshComponent}
 Overlay Mat Instance Dyanmic {type=MaterialInstanceDynamic}
 Pawn Velocity {type=Vector}
 R2 Velocity Transform Location {type=Vector}
 R2 Velocity Transform Velocity {type=Vector}
 RBAN Params {
  type=RBANParamsStruct
  default {
   dBNO {simSpaceSettings {
     maxAngularAcceleration=1000
     maxAngularVelocity=3
     maxLinearAcceleration=300
     maxLinearVelocity=700
     velocityScaleZ=0
     externalLinearDragV=<0 0.5 0>
    }}
   dBNO_Carried {simSpaceSettings {
     maxAngularAcceleration=1000
     maxAngularVelocity=3
     maxLinearAcceleration=300
     maxLinearVelocity=700
     velocityScaleZ=0
     externalLinearDragV=<0 0.5 0>
    }}
   emoteOrMelee {simSpaceSettings {
     maxAngularAcceleration=1000
     maxAngularVelocity=3
     maxLinearAcceleration=300
     maxLinearVelocity=700
     velocityScaleZ=0
     externalLinearDragV=<0 0.5 0>
    }}
   front_End {simSpaceSettings {
     maxAngularAcceleration=1000
     maxAngularVelocity=3
     maxLinearAcceleration=300
     maxLinearVelocity=700
     velocityScaleZ=0
     externalLinearDragV=<0 0.5 0>
    }}
   onGround_Crouch_Moving {simSpaceSettings {
     maxAngularAcceleration=1000
     maxAngularVelocity=3
     maxLinearAcceleration=300
     maxLinearVelocity=700
     velocityScaleZ=0
     externalLinearDragV=<0 0.5 0>
    }}
   onGround_Falling {simSpaceSettings {
     maxAngularAcceleration=1000
     maxAngularVelocity=3
     maxLinearAcceleration=300
     maxLinearVelocity=700
     velocityScaleZ=0
     externalLinearDragV=<0 0.5 0>
    }}
   onGround_Moving {simSpaceSettings {
     maxAngularAcceleration=1000
     maxAngularVelocity=3
     maxLinearAcceleration=300
     maxLinearVelocity=700
     velocityScaleZ=0
     externalLinearDragV=<0 0.5 0>
    }}
   onGround_Standing {simSpaceSettings {
     maxAngularAcceleration=1000
     maxAngularVelocity=3
     maxLinearAcceleration=300
     maxLinearVelocity=700
     velocityScaleZ=0
     externalLinearDragV=<0 0.5 0>
    }}
   rail_Grinding {simSpaceSettings {
     maxAngularAcceleration=1000
     maxAngularVelocity=3
     maxLinearAcceleration=300
     maxLinearVelocity=700
     velocityScaleZ=0
     worldAlpha=0.1
     externalLinearDragV=<0 0.5 0>
    }}
   riding_Moving {simSpaceSettings {
     maxAngularAcceleration=1000
     maxAngularVelocity=3
     maxLinearAcceleration=300
     maxLinearVelocity=700
     velocityScaleZ=0
     externalLinearDragV=<0 0.5 0>
    }}
   riding_Moving_Targeting {simSpaceSettings {
     maxAngularAcceleration=1000
     maxAngularVelocity=3
     maxLinearAcceleration=300
     maxLinearVelocity=700
     velocityScaleZ=0
     externalLinearDragV=<0 0.5 0>
    }}
   riding_Standing {simSpaceSettings {
     maxAngularAcceleration=1000
     maxAngularVelocity=3
     maxLinearAcceleration=300
     maxLinearVelocity=700
     velocityScaleZ=0
     externalLinearDragV=<0 0.5 0>
    }}
   skydive_DivingDown {simSpaceSettings {
     maxAngularAcceleration=1000
     maxAngularVelocity=3
     maxLinearAcceleration=300
     maxLinearVelocity=700
     velocityScaleZ=0
     externalLinearDragV=<0 0.5 0>
    }}
   skydive_FloatingDown {
    componentLinearAccScale=<0 0.5 0>
    componentLinearVelScale=<0 0.5 0>
    simSpaceSettings {
     maxAngularAcceleration=1000
     maxAngularVelocity=3
     maxLinearAcceleration=300
     maxLinearVelocity=700
     velocityScaleZ=0
     externalLinearDragV=<0 0.5 0>
    }
   }
   skydive_GoingUp {simSpaceSettings {
     maxAngularAcceleration=1000
     maxAngularVelocity=3
     maxLinearAcceleration=300
     maxLinearVelocity=700
     velocityScaleZ=0
     externalLinearDragV=<0 0.5 0>
    }}
   skydive_Parachuting {
    componentLinearAccScale=<0 0.1 0.3>
    componentLinearVelScale=<0 0.1 0.3>
    simSpaceSettings {
     maxAngularAcceleration=1000
     maxAngularVelocity=3
     maxLinearAcceleration=300
     maxLinearVelocity=700
     velocityScaleZ=0
     externalLinearDragV=<0 0.5 0>
    }
   }
   skydive_Surfing {simSpaceSettings {
     maxAngularAcceleration=1000
     maxAngularVelocity=3
     maxLinearAcceleration=300
     maxLinearVelocity=700
     velocityScaleZ=0
     externalLinearDragV=<0 0.5 0>
    }}
   swimming {simSpaceSettings {
     maxAngularAcceleration=1000
     maxAngularVelocity=3
     maxLinearAcceleration=300
     maxLinearVelocity=700
     velocityScaleZ=0
     externalLinearDragV=<0 0.5 0>
    }}
   tactical_Sprint {simSpaceSettings {
     maxAngularAcceleration=1000
     maxAngularVelocity=3
     maxLinearAcceleration=300
     maxLinearVelocity=700
     velocityScaleZ=0
     externalLinearDragV=<0 0.5 0>
    }}
   ziplining {simSpaceSettings {
     maxAngularAcceleration=1000
     maxAngularVelocity=3
     maxLinearAcceleration=300
     maxLinearVelocity=700
     velocityScaleZ=0
     externalLinearDragV=<0 0.5 0>
    }}
  }
 }
 Turbulence Noise {type=number}
 Velocity Interp Speed {type=number}
 WPO Wind {type=number}
 X Movement {type=number}
]
